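# Niagara water FX system for the TideCompost emote's board, drawing spray, splashes and wake as the character moves.
# Emitters including Edge_Spray_Mesh, Ribbons, Splash_Both and Wake_Mesh follow a sampled skeletal mesh, mirroring left and right through spawn groups or partitions.
# User RightDot, ForwardDot and Thruster shape lean, scale, orientation and velocity.
# Spray and splashes blend between world and local space by owner speed; time-of-day scalars tune brightness.
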NiagaraSystem /BRCosmetics/Animation/Game/MainPlayer/Emotes/TideCompost/FX/NS_Emote_TideCompost_Board_Splash
dependencies {
 usedDataInterfaces [
  &NiagaraDataInterfaceCurve
  &NiagaraDataInterfaceVectorCurve
  &NiagaraDataInterfaceSpriteRendererInfo
  &NiagaraDataInterfaceVector2DCurve
  &NiagaraDataInterfaceArrayFloat
 ]
 usedDynamicInputs []
 usedModules [
  &self:SystemSpawnScript.SystemScriptSource.SystemScriptGraph.NiagaraNodeAssignment_0.SetVariables_#1
  &SystemState
  &BlendSpaceHelper
  &SampleSkeletalMesh
  &self:SystemSpawnScript.SystemScriptSource.SystemScriptGraph.NiagaraNodeAssignment_2.SetVariables_#2
  &self:SystemSpawnScript.SystemScriptSource.SystemScriptGraph.NiagaraNodeAssignment_1.SetVariables_#3
 ]
}
emitters {
 Edge_Spray_Mesh {
  emitterUpdateScript [
   &EmitterState(
    Inactive Response=::NewEnumerator1
    Life Cycle Mode=::NewEnumerator1
    Loop Behavior=::NewEnumerator1
    Loop Duration Mode=::NewEnumerator1
    UseLoopDelay=false
   )
   &SpawnRate(SpawnRate=&FN_FrontEndFloatModifier(Gameplay=60, Front End=60), Use Spawn Probability=false)
   &SpawnRate(
    SpawnGroup=1
    SpawnRate=&FN_FrontEndFloatModifier(Gameplay=30, Front End=30)
    Use Spawn Probability=false
   ) {module=SpawnRate001}
  ]
  particleSpawnScript [
   &PartitionParticles(Partition Mode:ENiagara_PartitionMode_v2=::NewEnumerator5)
   /SetVariables(Particles.VisibilityTag=$Emitter.SpawnGroup) {isSetParameters=true}
   &FN_SetSpawnGroup(Number of Spawn Groups=2) {enabled=false}
   /SetVariables(
    Particles.FrontToBack:NiagaraFloat=&RandomRangeFloat(Minimum=0, Maximum=1)
    Particles.LeanLerp:NiagaraFloat=&Multiply_Float(
     A=$User.RightDot
     B=&Lerp_Float(A=1, B=-1, Alpha=&MakeFloatFromInt(Int=$Particles.PartitionParticles.Partition))
    )
   ) {isSetParameters=true}
   &InitializeParticle(
    Color=&MaskLinearColor(Mask Source=0, Mask Parameter=?, Mask=0, Masked Linear Color=<rgba 1 1 1 1>, Passthrough Linear Color=<rgba 0.5052 0.7402 1 0.5>)
    Lifetime=&MaskFloat(
     Mask Source=0
     Mask Parameter=?
     Mask=0
     Masked Float=&RandomRangeFloat(Minimum=0.5, Maximum=0.7)
     Passthrough Float=0.5
    )
    Mesh Scale Max=<1 0.75 1>
    Mesh Scale Min=<0.75 0.5 0.75>
    Mesh Scale Mode=::NewEnumerator2
    Sprite Rotation Mode=::NewEnumerator2
    Sprite Size Max=<50 50>
    Sprite Size Min=<20 20>
    Sprite Size Mode=::NewEnumerator2
    Sprite UV Mode=::NewEnumerator3
    UsePositionOffset=false
    Write Material Random=true
   ) {module=InitializeParticle001}
   &SkeletalMeshLocation(
    Apply Sampled Orientation to Mesh Particles=true
    Inherit Velocity=false
    Interpolate Mesh Sampling=false
    MaskSpawnGroup=false
    Mesh Sampling Type=::NewEnumerator5
    Sampled Position Offset=&VectorFromCurve(
     VectorCurve=DataInterface() {propertyValues {
       XCurve {keys [
         <at 0 -25>
         {arriveTangent=1.2394, interpMode=RCIM_Cubic, leaveTangent=1.2394, tangentMode=RCTM_User, time=0.33, value=-30}
         {arriveTangent=5.973, arriveTangentWeight=0.2233, interpMode=RCIM_Cubic, leaveTangent=5.973, tangentMode=RCTM_User, time=1, value=0}
        ]}
       YCurve {
        defaultValue=0
        keys [
         <at 0 30>
         {
          arriveTangent=-382.5781
          arriveTangentWeight=0.3333
          interpMode=RCIM_Cubic
          leaveTangent=-382.5781
          tangentMode=RCTM_User
          time=1
          value=-100
         }
        ]
       }
       ZCurve {keys [<at 0 -10>,<at 1 -15>]}
      }}
     CurveIndex=$Particles.FrontToBack
     Scale Curve=&Lerp_Vector(A=<1 1 1>, B=<-1 1 1>, Alpha=&MakeFloatFromInt(Int=$Particles.PartitionParticles.Partition))
    )
    Sampled Position Offset Coordinate Space:ENiagaraMeshTransforms=::NewEnumerator3
    Skeletal Mesh=$System.Skeletal Mesh Reader
    Socket Sampling Mode=::NewEnumerator5
   )
   &InitialMeshOrientation(
    Orientation Vector=&Multiply_Vector(
     A=&Add_Vector(
      A=&Lerp_Vector(A=<0 0 0>, B=<-1 0 0>, Alpha=$Particles.FrontToBack)
      B=&Lerp_Vector(A=<-1 0.8 0>, B=<-0.75 1 0>, Alpha=$Particles.LeanLerp)
     )
     B=&Multiply_Vector(
      A=&Lerp_Vector(A=<1 1 0>, B=<1 -1 1>, Alpha=&MakeFloatFromInt(Int=$Particles.PartitionParticles.Partition))
      B=&Lerp_Vector(A=<1 1 1>, B=<1 0.75 1>, Alpha=&ClampFloat(Float=$User.ForwardDot))
     )
    )
    Rotation Coordinate Space=::NewEnumerator0
    Use Orientation Vector=true
    Use Rotation Vector=false
   ) {enabled=false}
   &InitialMeshOrientation(
    Orientation Vector=&Multiply_Vector(
     A=&Add_Vector(
      A=&VectorFromCurve(
       VectorCurve=DataInterface() {propertyValues {
         XCurve {keys [<at 0.33 1 RCIM_Cubic>, {arriveTangent=-1.96, interpMode=RCIM_Cubic, leaveTangent=-1.96, tangentMode=RCTM_User, time=1, value=0.2}]}
         YCurve {keys [<at 0.33 0.5 RCIM_Cubic>, {arriveTangent=-0.6339, interpMode=RCIM_Cubic, leaveTangent=-0.6339, tangentMode=RCTM_User, time=1, value=0.2}]}
         ZCurve {keys [<at 0 0>,<at 1 0>]}
        }}
       CurveIndex=$Particles.FrontToBack
       Scale Curve=<-1 1 1>
      )
      B=&Lerp_Vector(A=<0 0 0>, B=<0.5 0 0>, Alpha=$Particles.LeanLerp)
     )
     B=&Multiply_Vector(
      A=&Lerp_Vector(A=<1 1 1>, B=<1 -1 1>, Alpha=&MakeFloatFromInt(Int=$Particles.SpawnGroup))
      B=&Lerp_Vector(A=<1 0 1>, B=<1 0 1>, Alpha=$User.ForwardDot)
     )
    )
    Rotation Coordinate Space=::NewEnumerator0
    Use Orientation Vector=true
    Use Rotation Vector=false
   ) {enabled=false, module=InitialMeshOrientation001}
   &InitialMeshOrientation(
    Orientation Vector=&Multiply_Vector(
     A=&Add_Vector(
      A=&Lerp_Vector(A=<-1 1 0>, B=<-1 -1 0>, Alpha=$Particles.FrontToBack)
      B=&Lerp_Vector(A=<-0.5 1 0>, B=<-1 1 0>, Alpha=$Particles.LeanLerp)
     )
     B=&Multiply_Vector(
      A=&Lerp_Vector(A=<1 1 1>, B=<1 -1 1>, Alpha=&MakeFloatFromInt(Int=$Particles.PartitionParticles.Partition))
      B=&Lerp_Vector(A=<1 0.8 1>, B=<1 0.2 1>, Alpha=&ClampFloat(Float=$User.Thruster))
     )
    )
    Rotation Coordinate Space=::NewEnumerator0
    Use Orientation Vector=true
    Use Rotation Vector=false
   ) {module=InitialMeshOrientation002}
   /SetVariables(Particles.Scale=&Multiply_Vector(
     A=&VectorFromCurve(
      VectorCurve=DataInterface() {propertyValues {
        XCurve {keys [{arriveTangent=1.7846, interpMode=RCIM_Cubic, leaveTangent=1.7846, tangentMode=RCTM_User, time=0, value=0.4}, <at 1 1 RCIM_Cubic>]}
        YCurve {keys [{arriveTangent=1.7846, interpMode=RCIM_Cubic, leaveTangent=1.7846, tangentMode=RCTM_User, time=0, value=0.4}, <at 1 1 RCIM_Cubic>]}
        ZCurve {keys [{arriveTangent=1.7846, interpMode=RCIM_Cubic, leaveTangent=1.7846, tangentMode=RCTM_User, time=0, value=0.4}, <at 1 1 RCIM_Cubic>]}
       }}
      CurveIndex=$Particles.FrontToBack
      Scale Curve=&Multiply_VectorByFloat(
       Vector=$Particles.Scale
       Float=&Lerp_Float(A=0.8, B=1.4, Alpha=&ClampFloat(Float=$User.ForwardDot))
      )
     )
     B=&Lerp_Vector(A=<0.6 0.6 0.6>, B=<0.8 0.8 0.8>, Alpha=$Particles.LeanLerp)
    )) {isSetParameters=true}
   &AddVelocity(
    Scale Added Velocity=&Lerp_Vector(
     A=<1 1 1>
     B=<1.1 1.1 1.25>
     Alpha=&Add_Float(A=&ClampFloat(Float=$User.ForwardDot), B=$Particles.LeanLerp)
    )
    Velocity=&MaskVector(
     Mask Source=0
     Mask Parameter=?
     Mask=0
     Masked Vector=<0 -1 0>
     Passthrough Vector=&RandomRangeVector(Minimum=<-10 -100 100>, Maximum=<10 -200 150>)
    )
   )
   &AddVelocityFromPoint(
    Module.Use Falloff Distance=false
    Origin Offset=<0 50 -5>
    Origin Offset Coordinate Space=::Simulation
    Velocity Strength=&MaskFloat(
     Mask Source=0
     Mask Parameter=?
     Mask=0
     Masked Float=0
     Passthrough Float=&RandomRangeFloat(Minimum=100, Maximum=100)
    )
   )
  ]
  particleUpdateScript [
   &ParticleState()
   &ScaleSpriteSize(Scale Factor=&Vector2DFromFloat(Value=&FloatFromCurve(FloatCurve=DataInterface() {propertyValues {Curve {keys [
          {arriveTangent=1.6115, interpMode=RCIM_Cubic, leaveTangent=1.6115, tangentMode=RCTM_User, time=0, value=0.1}
          {interpMode=RCIM_Cubic, tangentMode=RCTM_User, time=1, value=1}
         ]}}})))
   &SpriteRotationRate(Rotation Rate=&RandomRangeFloat(Minimum=-90, Maximum=90))
   /SetVariables(Particles.Position=&AddVectorToPosition(
     Position=$System.SurfBoardPosition
     Vector=&VectorFromCurve(
      VectorCurve=DataInterface() {propertyValues {
        XCurve {keys [
          {interpMode=RCIM_Cubic, tangentMode=RCTM_User, time=0, value=-20}
          {arriveTangent=20, interpMode=RCIM_Cubic, leaveTangent=20, time=0.33, value=-30}
          {arriveTangent=5.973, arriveTangentWeight=0.2233, interpMode=RCIM_Cubic, leaveTangent=5.973, tangentMode=RCTM_User, time=1, value=0}
         ]}
        YCurve {
         defaultValue=0
         keys [
          <at 0 50>
          {
           arriveTangent=-382.5781
           arriveTangentWeight=0.3333
           interpMode=RCIM_Cubic
           leaveTangent=-382.5781
           tangentMode=RCTM_User
           time=1
           value=-100
          }
         ]
        }
        ZCurve {keys [<at 0 -5>,<at 1 -15>]}
       }}
      CurveIndex=$Particles.FrontToBack
      Scale Curve=&Lerp_Vector(A=<1 1 1>, B=<-1 1 1>, Alpha=&MakeFloatFromInt(Int=$Particles.SpawnGroup))
     )
    )) {enabled=false, isSetParameters=true}
   /SetVariables(Particles.Scale=&Multiply_Vector(
     A=$Particles.Initial.Scale
     B=&VectorFromCurve(
      VectorCurve=DataInterface() {propertyValues {
        XCurve {keys [
          {arriveTangent=4.6994, interpMode=RCIM_Cubic, leaveTangent=4.6994, tangentMode=RCTM_User, time=0, value=0.0795}
          {arriveTangent=0.4589, interpMode=RCIM_Cubic, leaveTangent=0.4589, tangentMode=RCTM_User, time=0.1447, value=0.7609}
          <at 1 1>
         ]}
        YCurve {keys [{arriveTangent=2.4365, interpMode=RCIM_Cubic, leaveTangent=2.4365, tangentMode=RCTM_User, time=0, value=0.01}, <at 1 1.5>]}
        ZCurve {keys [
          <at 0 0 RCIM_Cubic>
          {arriveTangent=-2.3619, interpMode=RCIM_Cubic, leaveTangent=-2.3619, tangentMode=RCTM_User, time=0.336, value=0.9567}
          <at 0.9988 0>
         ]}
       }}
      Scale Curve=&MakeVector(
       X=1
       Y=&FloatFromCurve(
        FloatCurve=DataInterface() {propertyValues {Curve {keys [{arriveTangent=1.184, interpMode=RCIM_Cubic, leaveTangent=1.184, tangentMode=RCTM_User, time=0, value=1}, <at 1 1.5>]}}}
        CurveIndex=$Particles.FrontToBack
       )
       Z=0.8
      )
     )
    )) {isSetParameters=true}
   &SkeletalMeshLocation(
    Apply To Particle Position:NiagaraBool=/Niagara/DynamicInputs/Helpers/MaskBoolBySpawnGroup(Masked Bool:NiagaraBool=false, Passthrough Bool:NiagaraBool=true, Spawn Group Mask:NiagaraInt32=1)
    Inherit Velocity=false
    MaskSpawnGroup=false
    Mesh Sampling Type=::NewEnumerator5
    Sampled Position Offset=&VectorFromCurve(
     VectorCurve=DataInterface() {propertyValues {
       XCurve {keys [
         {interpMode=RCIM_Cubic, leaveTangentWeight=0.11, tangentMode=RCTM_User, time=0, value=-20}
         {
          arriveTangentWeight=0.1758
          interpMode=RCIM_Cubic
          leaveTangentWeight=0.6462
          tangentMode=RCTM_Break
          tangentWeightMode=RCTWM_WeightedBoth
          time=0.33
          value=-35
         }
         {arriveTangent=102.4641, arriveTangentWeight=0.2233, interpMode=RCIM_Cubic, leaveTangent=102.4641, tangentMode=RCTM_User, time=1, value=0}
        ]}
       YCurve {
        defaultValue=0
        keys [
         {interpMode=RCIM_Cubic, tangentMode=RCTM_User, time=0, value=50}
         {
          arriveTangent=-382.5781
          arriveTangentWeight=0.3333
          interpMode=RCIM_Cubic
          leaveTangent=-382.5781
          tangentMode=RCTM_User
          time=1
          value=-120
         }
        ]
       }
       ZCurve {keys [<at 0 -10>,<at 1 0>]}
      }}
     CurveIndex=$Particles.FrontToBack
     Scale Curve=&Lerp_Vector(A=<1 1 1>, B=<-1 1 1>, Alpha=&MakeFloatFromInt(Int=$Particles.PartitionParticles.Partition))
    )
    Sampled Position Offset Coordinate Space:ENiagaraMeshTransforms=::NewEnumerator3
    Skeletal Mesh=$System.Skeletal Mesh Reader
    Socket Sampling Mode=::NewEnumerator5
   ) {enabled=false, module=SkeletalMeshLocation001}
   &SkeletalMeshLocation(
    MaskSpawnGroup=false
    Mesh Sampling Type=::NewEnumerator5
    Orientation Sampling=::NewEnumerator1
    Position Sampling=::NewEnumerator1
    Sampled Position Offset=&VectorFromCurve(
     VectorCurve=DataInterface() {propertyValues {
       XCurve {keys [
         {interpMode=RCIM_Cubic, leaveTangentWeight=0.11, tangentMode=RCTM_User, time=0, value=-20}
         {
          arriveTangentWeight=0.1758
          interpMode=RCIM_Cubic
          leaveTangentWeight=0.6462
          tangentMode=RCTM_Break
          tangentWeightMode=RCTWM_WeightedBoth
          time=0.33
          value=-35
         }
         {arriveTangent=102.4641, arriveTangentWeight=0.2233, interpMode=RCIM_Cubic, leaveTangent=102.4641, tangentMode=RCTM_User, time=1, value=0}
        ]}
       YCurve {
        defaultValue=0
        keys [
         {interpMode=RCIM_Cubic, tangentMode=RCTM_User, time=0, value=50}
         {
          arriveTangent=-382.5781
          arriveTangentWeight=0.3333
          interpMode=RCIM_Cubic
          leaveTangent=-382.5781
          tangentMode=RCTM_User
          time=1
          value=-120
         }
        ]
       }
       ZCurve {keys [<at 0 -10>,<at 1 0>]}
      }}
     CurveIndex=$Particles.FrontToBack
     Scale Curve=&Lerp_Vector(A=<1 1 1>, B=<-1 1 1>, Alpha=&MakeFloatFromInt(Int=$Particles.PartitionParticles.Partition))
    )
    Skeletal Mesh=$System.Skeletal Mesh Reader
    Socket Sampling Mode=::NewEnumerator5
   ) {enabled=false, module=SkeletalMeshLocation002}
   &GravityForce(Gravity=<0 0 -900>) {enabled=false}
   &Drag(Drag=0.5, Use Linear Drag=true, Use Rotational Drag=false)
   &AccelerationForce(Acceleration=&MaskVector(Mask Source=0, Mask Parameter=?, Mask=0, Masked Vector=<0 -500 0>, Passthrough Vector=<0 0 -980>)) {enabled=false}
   &InheritSourceMovement(
    Applied Position Scale=&TransformVector(
     Vector=&Multiply_VectorByFloat(
      Vector=<1 1 1>
      Float=&FloatFromCurve(FloatCurve=DataInterface() {propertyValues {Curve {keys [{interpMode=RCIM_Cubic, tangentMode=RCTM_User, time=0.5, value=1}, <at 1 0>]}}})
     )
     Source Space=::World
     Destination Space=::Local
    )
    ApplyForce=false
    ApplyPosition=true
    ApplyPositionFalloff=false
    Source Velocity:Vector3f=$System.SampleSkeletalMesh.SampledVelocity
   ) {enabled=false}
   &SolveForcesAndVelocity(Clamp Velocity=false, Limit Acceleration=false)
   &DynamicMaterialParameters(
    Index 0 Param 1=&FloatFromCurve(FloatCurve=DataInterface() {propertyValues {Curve {keys [
         {arriveTangent=1.2629, interpMode=RCIM_Cubic, leaveTangent=1.2629, tangentMode=RCTM_User, time=-0.0045, value=-1.9799}
         {arriveTangent=4.4591, interpMode=RCIM_Cubic, leaveTangent=4.4591, tangentMode=RCTM_User, time=1, value=0}
        ]}}})
    Index 0 Param 2=&FloatFromCurve(FloatCurve=DataInterface() {propertyValues {Curve {keys [<at 0.0102 4>,<at 0.9749 0>]}}})
    Index 0 Param 3=&MakeFloatFromVector(Vector=$Particles.Scale, Channel=::NewEnumerator0)
    Index 0 Param 4=1
    Index 1 Param 1=&FloatFromCurve(FloatCurve=DataInterface() {propertyValues {Curve {keys [<at 0.7 0 RCIM_Cubic>, {arriveTangent=6.4641, interpMode=RCIM_Cubic, leaveTangent=6.4641, tangentMode=RCTM_User, time=1, value=1}]}}})
    Module.Param0WriteEnabled=true
    Module.Param0WriteEnabled1=true
    Module.Param1WriteEnabled=true
    Module.Param1WriteEnabled1=true
    Module.Param2WriteEnabled=true
    Module.Param2WriteEnabled1=true
    Module.Param3WriteEnabled=true
    Module.Param3WriteEnabled1=true
    Write Parameter Index 1=true
   )
   &ScaleColor(
    Scale Alpha=&FloatFromCurve(
     FloatCurve=DataInterface() {propertyValues {Curve {keys [{interpMode=RCIM_Cubic, tangentMode=RCTM_User, time=0, value=0}, {interpMode=RCIM_Cubic, tangentMode=RCTM_User, time=0.16, value=1}]}}}
     Scale Curve=&FNTimeOfDayScalar(Night Scale=0.8, Morning Scale=1, Evening Scale=0.9)
    )
    Scale RGB=&MaskVector(
     Mask Source=0
     Mask Parameter=?
     Mask=0
     Masked Vector=<1 1 1>
     Passthrough Vector=&VectorFromFloat(Value=&RandomRangeFloat(Minimum=1, Maximum=5))
    )
    ScaleA=true
    ScaleRGB=true
   )
   /SetVariables(Particles.Splash DynP 00:Vector4f=&MakeVector4(
     X=&RandomRangeFloat(Minimum=0, Maximum=1)
     Y=0
     Z=&FloatFromCurve(FloatCurve=DataInterface() {propertyValues {Curve {keys [{interpMode=RCIM_Cubic, tangentMode=RCTM_User, time=0, value=0.5}, {interpMode=RCIM_Cubic, tangentMode=RCTM_User, time=1, value=2}]}}})
     W=&FloatFromCurve(FloatCurve=DataInterface() {propertyValues {Curve {keys [<at 0 0 RCIM_Cubic>, {arriveTangent=1.5276, interpMode=RCIM_Cubic, leaveTangent=1.5276, tangentMode=RCTM_User, time=1, value=1}]}}})
    )) {enabled=false, isSetParameters=true}
   /SetVariables(Particles.Splash DynP 00:Vector4f=&MakeVector4_FromFloat(Float=&FloatFromCurve(FloatCurve=DataInterface() {propertyValues {Curve {keys [{interpMode=RCIM_Cubic, tangentMode=RCTM_Break, time=0.6, value=1}, {interpMode=RCIM_Cubic, tangentMode=RCTM_Break, time=1, value=0.4}]}}}))) {isSetParameters=true}
   &BlendSpace(
    "Mesh Orientation Blend (World <-> Local)":NiagaraFloat=&MakeCustomFloatFromBool(
     Bool=&SetBoolByFloatComparison(A=&VectorLength(Vector=$Engine.Owner.Velocity), B=350)
     True Float=1
     False Float=0.3
    )
    Mesh Orientation Mode=1
    "Position Blend (World <-> Local)":NiagaraFloat=&MakeCustomFloatFromBool(
     Bool=&SetBoolByFloatComparison(A=&VectorLength(Vector=$Engine.Owner.Velocity), B=350)
     True Float=1
     False Float=&FloatFromCurve(FloatCurve=DataInterface() {propertyValues {Curve {keys [<at 0 1>,<at 0.6 0.4>]}}})
     Translation Contribution:NiagaraFloat=1
     Rotation Contribution:NiagaraFloat=1
    )
   )
  ]
  rendererClasses [&NiagaraMeshRendererProperties, &NiagaraSpriteRendererProperties]
  renderers [
   {rendererClass=&NiagaraMeshRendererProperties, rendererIndex=0}
   {rendererClass=&NiagaraMeshRendererProperties, rendererIndex=1}
   {rendererClass=&NiagaraSpriteRendererProperties, rendererIndex=2}
   {rendererClass=&NiagaraSpriteRendererProperties, rendererIndex=3}
  ]
 }
 Ribbons {
  emitterUpdateScript [
   &EmitterState(
    Inactive Response=::NewEnumerator1
    Life Cycle Mode=::NewEnumerator1
    Loop Behavior=::NewEnumerator1
    Loop Delay=1.5
    Loop Duration Mode=::NewEnumerator1
    UseLoopDelay=false
   )
   &SpawnBurst_Instantaneous(Spawn Count=2, Spawn Time=0.5, Use Spawn Probability=false, UseLoopCountLimit=false)
   &SpawnBurst_Instantaneous(Spawn Count=2, Spawn Group=1, Spawn Time=0.5, Use Spawn Probability=false, UseLoopCountLimit=false) {module=SpawnBurst_Instantaneous001}
  ]
  particleSpawnScript [
   /SetVariables(Particles.InitialSpawnGroup:NiagaraInt32=&SpawnGroup()) {isSetParameters=true}
   &FN_SetSpawnGroup()
   /SetVariables(Particles.VisibilityTag=&MergeHelper_IntSwitch(0=1, 1=0, 2=0, 3=1, 4=?, 5=?, Index=$Particles.SpawnGroup, Num=::NewEnumerator2)) {isSetParameters=true}
   &PartitionParticles()
   &InitializeParticle(
    Color=<rgba 1 1 1 0.1>
    Lifetime=0.2
    Mesh Scale=&Multiply_VectorByFloat(
     Vector=&MakeVector(
      X=&SelectFloatFromArray(Direct Array Index=$Particles.InitialSpawnGroup, Float Selection Array=DataInterface() {propertyValues {FloatData [1,-1]}})
      Y=1
      Z=1
     )
     Float=0.5
    )
    Mesh Scale Mode=::NewEnumerator4
    UsePositionOffset=false
    Write Material Random=true
   ) {module=InitializeParticle001}
   &SocketLocation(
    SamplePosOffset=true
    Sampled Position Offset=&MergeHelper_VectorSwitch(0=<2 50 -40>, 1=<-2 50 -40>, 2=?, 3=?, Index=$Particles.PartitionParticles.Partition, Num=::NewEnumerator0)
    Socket Reader=$System.Socket Reader
    Socket Sampling Mode=::NewEnumerator3
    VelocitySpeedLimit=false
   ) {enabled=false}
   &CameraOffset(Camera Offset Amount=5)
  ]
  particleUpdateScript [
   &ParticleState(Kill Particles When Lifetime Has Elapsed=false)
   &SkeletalMeshLocation(
    MaskSpawnGroup=false
    Mesh Sampling Type=::NewEnumerator5
    Orientation Sampling=::NewEnumerator1
    Sampled Position Offset=&MergeHelper_VectorSwitch(0=<10 35 10>, 1=<-10 35 10>, 2=?, 3=?, Index=$Particles.PartitionParticles.Partition, Num=::NewEnumerator0)
    Skeletal Mesh=$System.Skeletal Mesh Reader
    Socket Sampling Mode=::NewEnumerator5
   ) {module=SkeletalMeshLocation001}
   &SolveForcesAndVelocity(Clamp Velocity=false, Limit Acceleration=false)
   &InterpolateOverTime(
    Target Vector Value:Vector3f=&TransformPosition(Position=$Particles.Position, Source Space=::Local, Destination Space=::World)
    Variable Type=::NewEnumerator2
   )
   &DynamicMaterialParameters(
    Index 0 Param 1=&RandomRangeFloat(Minimum=0, Maximum=1)
    Index 0 Param 3=&MakeFloatFromVector(Vector=<0 0 0>, Channel=::NewEnumerator0)
    Index 1 Param 1=&MakeFloatFromVector(Vector=$StackContext.InterpolateOverTime.InterpolateOverTime_V3_.Moving Average, Channel=::NewEnumerator0)
    Index 1 Param 2=&MakeFloatFromVector(Vector=$StackContext.InterpolateOverTime.InterpolateOverTime_V3_.Moving Average, Channel=::NewEnumerator1)
    Index 1 Param 3=&Add_Float(
     A=&MakeFloatFromVector(Vector=$StackContext.InterpolateOverTime.InterpolateOverTime_V3_.Moving Average, Channel=::NewEnumerator2)
     B=&FloatFromCurve(
      FloatCurve=DataInterface() {propertyValues {Curve {
         defaultValue=1
         keys [
          {arriveTangent=10.3639, interpMode=RCIM_Cubic, leaveTangent=10.3639, leaveTangentWeight=0.2653, tangentMode=RCTM_User, time=0, value=0}
          {arriveTangent=-0.474, interpMode=RCIM_Cubic, leaveTangent=-0.474, time=0.2766, value=1}
          {
           arriveTangent=-1.1179
           arriveTangentWeight=0.3979
           interpMode=RCIM_Cubic
           leaveTangent=-1.1179
           leaveTangentWeight=0.1021
           tangentMode=RCTM_User
           time=0.7958
           value=-0.3772
          }
          {arriveTangent=-4.1785, arriveTangentWeight=0.2925, interpMode=RCIM_Cubic, leaveTangent=-4.1785, tangentMode=RCTM_User, time=1, value=-1}
         ]
        }}}
      Scale Curve=100
     )
    )
    Module.Index0Float0Write=true
    Module.Index0Float1Write=true
    Module.Index0Float2Write=true
    Module.Index0Float3Write=true
    Module.Index1Float0Write=true
    Module.Index1Float1Write=true
    Module.Index1Float2Write=true
    Module.Index1Float3Write=false
    Write Parameter Index 1=::NewEnumerator1
   ) {module=DynamicMaterialParameters001}
   &ScaleColor(
    Scale Alpha=&FloatFromCurve(FloatCurve=DataInterface() {propertyValues {Curve {keys [
         {arriveTangent=2.8784, interpMode=RCIM_Cubic, leaveTangent=2.8784, tangentMode=RCTM_User, time=0, value=0}
         {interpMode=RCIM_Cubic, tangentMode=RCTM_User, time=0.3, value=1}
        ]}}})
    Scale RGB=&MakeVectorFromLinearColorRGB(Color=&FNTimeOfDayColor(Night Color=<rgba 0.334 0.556 1 1>, Morning Color=<rgba 1 1 1 1>, Day Color=<rgba 1 1 1 1>, Evening Color=<rgba 1 1 1 1>))
    ScaleA=true
    ScaleRGB=true
   )
  ]
  rendererClasses [&NiagaraMeshRendererProperties]
  renderers [
   {rendererClass=&NiagaraMeshRendererProperties, rendererIndex=0}
   {rendererClass=&NiagaraMeshRendererProperties, rendererIndex=1}
  ]
 }
 Splash_Both {
  emitterUpdateScript [
   &EmitterState()
   &SpawnRate(SpawnRate=15, Use Spawn Probability=false)
   &SpawnRate(SpawnGroup=1, SpawnRate=15, Use Spawn Probability=false) {module=SpawnRate001}
  ]
  particleSpawnScript [
   &PartitionParticles(Partition Mode:ENiagara_PartitionMode_v2=::NewEnumerator5)
   /SetVariables(
    Particles.SplashSpeed:NiagaraFloat=&MaskFloat(Mask Source=1, Mask Parameter=$Particles.PartitionParticles.Partition, Mask=0, Masked Float=0, Passthrough Float=1)
    Particles.VisibilityTag=$Particles.PartitionParticles.Partition
   ) {isSetParameters=true}
   &InitializeParticle(
    Color=&RandomRangeLinearColor(Minimum=<rgba 0.5052 0.7402 1 0.5>, Maximum=<rgba 0.2252 0.394 0.6906 0.7>)
    Lifetime Max=0.6
    Lifetime Min=0.5
    Lifetime Mode=::NewEnumerator1
    Sprite Rotation Angle=&MaskFloat(
     Mask Source=1
     Mask Parameter=$Particles.PartitionParticles.Partition
     Mask=0
     Masked Float=&RandomRangeFloat(Minimum=0, Maximum=1)
     Passthrough Float=0
    )
    Sprite Rotation Mode=::NewEnumerator0
    Sprite Size=/Niagara/DynamicInputs/Helpers/MaskVector2d(
     Mask Source:NiagaraInt32=1
     Mask Parameter:NiagaraInt32=$Particles.PartitionParticles.Partition
     Mask:NiagaraInt32=0
     Masked Vector2D:Vector2f=&RandomRangeVector2D(Minimum=<30 30>, Maximum=<50 50>)
     Passthrough Vector2D:Vector2f=&RandomRangeVector2D(Minimum=<18 60>, Maximum=<30 75>)
    )
    Sprite Size Mode=::NewEnumerator4
    Sprite UV Mode=::NewEnumerator3
    UsePositionOffset=false
    Write Material Random=true
   )
   &SkeletalMeshLocation(
    MaskSpawnGroup=false
    Mesh Sampling Type=::NewEnumerator5
    Orientation Sampling=::NewEnumerator1
    Sampled Position Offset=<0 0 -10>
    Skeletal Mesh=$System.Skeletal Mesh Reader
    Socket Sampling Mode=::NewEnumerator5
   ) {module=SkeletalMeshLocation001}
   &ShapeLocation(
    Box Size=<0 15 0>
    MaskSpawnGroup=true
    Offset=<30 15 3>
    Offset Mode=::NewEnumerator0
    Override Seed=false
    Rotation Mode=::NewEnumerator2
    Shape Primitive=::NewEnumerator2
    Surface Only=false
    UseOwnerScale=true
    Yaw / Pitch / Roll=<0.1 0 0>
   )
   &ShapeLocation(
    Box Size=<0 15 0>
    MaskSpawnGroup=true
    Offset=<-30 15 3>
    Offset Mode=::NewEnumerator0
    Override Seed=false
    Rotation Mode=::NewEnumerator2
    Shape Primitive=::NewEnumerator2
    Spawn Group Mask=1
    Surface Only=false
    UseOwnerScale=true
    Yaw / Pitch / Roll=<-0.1 0 0>
   ) {module=ShapeLocation001}
   &AddVelocity(Velocity=&Multiply_Vector(
     A=&Lerp_Vector(
      A=&RandomRangeVector(Minimum=<30 -100 90>, Maximum=<50 -150 100>)
      B=&RandomRangeVector(Minimum=<30 -200 80>, Maximum=<50 -300 90>)
      Alpha=$Particles.SplashSpeed
     )
     B=&MaskVector(Mask Source=0, Mask Parameter=?, Mask=0, Masked Vector=<1 1 1>, Passthrough Vector=<-1 1 1>)
    ))
   &AddVelocity(
    Velocity=&MaskVector(
     Mask Source=0
     Mask Parameter=?
     Mask=0
     Masked Vector=&RandomRangeVector(Minimum=<30 -300 50>, Maximum=<50 -400 70>)
     Passthrough Vector=&MaskVector(
      Mask Source=0
      Mask Parameter=?
      Mask=1
      Masked Vector=&RandomRangeVector(Minimum=<-30 -300 50>, Maximum=<-50 -400 70>)
      Passthrough Vector=<1 1 1>
     )
    )
    Velocity Speed Scale=$Particles.SplashSpeed
   ) {enabled=false, module=AddVelocity001}
   &SubUVAnimation(
    End Frame Range Override=3
    Evaluation Type=::NewEnumerator0
    "Sprite Renderer "=DataInterface() {propertyValues {SpriteRenderer=&self:NiagaraEmitter_2.NiagaraSpriteRendererProperties_1}}
    SubUV Animation Mode:ENiagara_SubUVLookupModeV2=::NewEnumerator2
    UseEndFrame=true
    UseStartFrame=true
   )
   &InheritVelocity(Speed Limit=true, UseSpeedThreshold=false, Velocity Source=$Particles.SkeletalMeshLocation001.SampledVelocity)
   &CameraOffset(Camera Offset Amount=3)
  ]
  particleUpdateScript [
   &ParticleState()
   &SpriteRotationRate(Rotation Rate=&MaskFloat(
     Mask Source=0
     Mask Parameter=?
     Mask=0
     Masked Float=&RandomRangeFloat(Minimum=-70, Maximum=-100)
     Passthrough Float=&MaskFloat(
      Mask Source=0
      Mask Parameter=?
      Mask=1
      Masked Float=&RandomRangeFloat(Minimum=70, Maximum=100)
      Passthrough Float=1
     )
    ))
   &GravityForce(Gravity=&RandomRangeVector(Minimum=<0 0 -200>, Maximum=<0 0 -400>))
   &Drag(Drag=&RandomRangeFloat(Minimum=0.2, Maximum=0.3), Use Linear Drag=true, Use Rotational Drag=false)
   &SolveForcesAndVelocity(Clamp Velocity=false, Limit Acceleration=false)
   &ScaleSpriteSize(Uniform Curve Sprite Scale=DataInterface() {propertyValues {Curve {keys [
        {arriveTangent=4.1785, interpMode=RCIM_Cubic, leaveTangent=4.1785, tangentMode=RCTM_User, time=0, value=0.15}
        {arriveTangent=1.1179, interpMode=RCIM_Cubic, leaveTangent=1.1179, tangentMode=RCTM_User, time=0.2042, value=0.6228}
        {arriveTangent=0.0224, interpMode=RCIM_Cubic, leaveTangent=0.0224, tangentMode=RCTM_User, time=1, value=1}
       ]}}})
   /SetVariables(Particles.VelocityAlignedSize:Vector2f=&Multiply_Vector2D(
     A=$Particles.SpriteSize
     B=&Vector2DFromCurve(
      Vector2Curve=DataInterface() {propertyValues {XCurve {keys [<at 0 1>,<at 1 1>]}, YCurve {keys [<at 0 2 RCIM_Cubic>,<at 1 0.5>]}}}
      ScaleCurve=<1 1>
     )
    )) {isSetParameters=true}
   &ScaleColor(
    Scale Alpha=&FloatFromCurve(
     FloatCurve=DataInterface() {propertyValues {Curve {keys [
         {arriveTangent=-0.2723, interpMode=RCIM_Cubic, leaveTangent=14.5791, tangentMode=RCTM_Break, time=0, value=0.5}
         {interpMode=RCIM_Cubic, tangentMode=RCTM_Break, time=0.01, value=1}
         {arriveTangent=0.0378, interpMode=RCIM_Cubic, leaveTangent=0.0378, tangentMode=RCTM_User, time=0.8, value=1}
         {arriveTangent=-3.7547, interpMode=RCIM_Cubic, leaveTangent=-3.7547, tangentMode=RCTM_Break, time=1, value=0}
        ]}}}
     Scale Curve=&FNTimeOfDayScalar(Night Scale=0.4, Morning Scale=0.8, Day Scale=0.8, Evening Scale=0.7)
    )
    Scale RGB=&VectorFromFloat(Value=5)
    ScaleA=true
    ScaleRGB=true
   )
   &DynamicMaterialParameters(
    Index 0 Param 1=&FloatFromCurve(FloatCurve=DataInterface() {propertyValues {Curve {keys [{interpMode=RCIM_Cubic, tangentMode=RCTM_Break, time=0.6, value=1}, {interpMode=RCIM_Cubic, tangentMode=RCTM_Break, time=1, value=0.4}]}}})
    Module.Index0Float0Write=true
    Module.Index0Float1Write=false
    Module.Index0Float2Write=false
    Module.Index0Float3Write=false
   )
   &BlendSpace(
    "Mesh Orientation Blend (World <-> Local)":NiagaraFloat=&MakeCustomFloatFromBool(
     Bool=&SetBoolByFloatComparison(A=&VectorLength(Vector=$Engine.Owner.Velocity), B=350)
     True Float=1
     False Float=0.3
    )
    Mesh Orientation Mode=1
    "Position Blend (World <-> Local)":NiagaraFloat=&MakeCustomFloatFromBool(
     Bool=&SetBoolByFloatComparison(A=&VectorLength(Vector=$Engine.Owner.Velocity), B=350)
     True Float=1
     False Float=&FloatFromCurve(FloatCurve=DataInterface() {propertyValues {Curve {keys [<at 0 1>,<at 0.6 0.4>]}}})
     Translation Contribution:NiagaraFloat=1
     Rotation Contribution:NiagaraFloat=1
    )
   )
  ]
  rendererClasses [&NiagaraSpriteRendererProperties]
  renderers [
   {rendererClass=&NiagaraSpriteRendererProperties}
   {rendererClass=&NiagaraSpriteRendererProperties, rendererIndex=1}
  ]
 }
 Wake_Mesh {
  emitterUpdateScript [
   &EmitterState(Life Cycle Mode=::NewEnumerator1, Loop Duration=0.25, UseLoopDelay=false)
   &SpawnRate(
    SpawnRate=&Lerp_Float(A=9, B=9, Alpha=&ClampFloat(Float=$User.ForwardDot))
    Use Spawn Probability=false
   )
   &SpawnRate(SpawnGroup=1, SpawnRate=5, Use Spawn Probability=false) {module=SpawnRate001}
  ]
  particleSpawnScript [
   /SetVariables(Particles.VisibilityTag=&SpawnGroup()) {isSetParameters=true}
   /Niagara/Modules/Spawn/Initialization/InitializeParticle(
    Color=&MergeHelper_ColorSwitch(0=<rgba 1 1 1 1>, 1=<rgba 0.4 0.54 1 1>, 2=?, 3=?, Index=&SpawnGroup(), Num=::NewEnumerator0)
    Lifetime=&MergeHelper_FloatSwitch(
     0=0.75
     1=&RandomRangeFloat(Minimum=0.3, Maximum=0.4)
     2=?
     3=?
     Index=&SpawnGroup()
     Num=::NewEnumerator0
    )
    Mesh Scale=&MergeHelper_VectorSwitch(
     0=&UniformRangedVector(Minimum=<1.25 0.3 0.8>, Maximum=<2 0.5 1.15>)
     1=<3 1 -0.1>
     2=?
     3=?
     Index=&SpawnGroup()
     Num=::NewEnumerator0
    )
    Module.Write Color=true
    Module.Write Lifetime=true
    Module.Write Mass=true
    Module.Write Material Random=true
    Module.Write Position=false
    Module.Write Scale=true
    Module.Write SpriteRotation=false
    Module.Write SpriteSize=false
    Position=&Add_Vector(
     A=/Niagara/DynamicInputs/Helpers/SimulationPosition()
     B=&Multiply_VectorByFloat(Vector=$Engine.Owner.SystemXAxis, Float=-150)
    )
   )
   &InitialMeshOrientation(
    Mesh Axis To Orient=&MergeHelper_VectorSwitch(0=<1 0 0>, 1=<0 1 0>, 2=?, 3=?, Index=&SpawnGroup(), Num=::NewEnumerator0)
    Mesh Orientation Mode=::NewEnumerator2
    Orientation Coordinate Space=::Local
    Rotation=&MakeVector(X=0, Y=0, Z=&RandomRangeFloat(Minimum=-0.01, Maximum=0.01))
    Use Rotation Vector=true
   )
   &SkeletalMeshLocation(
    MaskSpawnGroup=false
    Mesh Sampling Type=::NewEnumerator5
    Sampled Position Offset=&MergeHelper_VectorSwitch(0=<0 0 -15>, 1=<0 0 0>, 2=?, 3=?, Index=&SpawnGroup(), Num=::NewEnumerator0)
    Skeletal Mesh=$System.Skeletal Mesh Reader
    Socket Sampling Mode=::NewEnumerator5
   )
   &AddVelocity(Velocity=&Multiply_VectorByFloat(
     Vector=&MergeHelper_VectorSwitch(
      0=<0 -100 0>
      1=&Lerp_Vector(A=<0 -500 0>, B=<0 -600 0>, Alpha=$User.Thruster)
      2=?
      3=?
      Index=&SpawnGroup()
      Num=::NewEnumerator0
     )
     Float=&Lerp_Float(A=1, B=5, Alpha=$User.ForwardDot)
    ))
   /SetVariables(Particles.Scale=&Multiply_Vector(
     A=$Particles.Scale
     B=&Lerp_Vector(A=<0.55 1 0.25>, B=<1 1 0.5>, Alpha=&ClampFloat(Float=$User.ForwardDot))
    )) {isSetParameters=true}
   /SetVariables(Particles.Ripples DynP 00:Vector4f=&MakeVector4(
     X=&RandomRangeFloat(Minimum=0, Maximum=0.1)
     Y=&RandomRangeFloat(Minimum=0, Maximum=1)
     Z=&RandomRangeFloat(Minimum=0, Maximum=1)
     W=0
    )) {isSetParameters=true}
  ]
  particleUpdateScript [
   &ParticleState()
   &SkeletalMeshLocation(
    MaskSpawnGroup=false
    Mesh Sampling Type=::NewEnumerator5
    Orientation Sampling=::NewEnumerator1
    Position Sampling=::NewEnumerator1
    Sampled Position Offset=&MergeHelper_VectorSwitch(0=<0 50 -10>, 1=<0 10 -10>, 2=?, 3=?, Index=&SpawnGroup(), Num=::NewEnumerator0)
    Skeletal Mesh=$System.Skeletal Mesh Reader
    Socket Sampling Mode=::NewEnumerator5
   ) {module=SkeletalMeshLocation001}
   &InheritSourceMovement(
    Applied Force Scale=&Lerp_Vector(
     A=<1 1 1>
     B=<0 0 0>
     Alpha=&FloatFromCurve(FloatCurve=DataInterface() {propertyValues {Curve {keys [
          {
           arriveTangentWeight=0.6137
           interpMode=RCIM_Cubic
           leaveTangentWeight=0.6137
           tangentMode=RCTM_User
           tangentWeightMode=RCTWM_WeightedBoth
           time=0
           value=1
          }
          <at 1 0>
         ]}}})
    )
    Applied Force Speed Limit=200
    Applied Position Scale=<0.25 0.25 0>
    ApplyForce=true
    ApplyForceFalloff=false
    ApplyPosition=false
    Module.LimitForceSpeed=false
   )
   /SetVariables(Particles.RipplesScale:Vector3f=&Multiply_Vector(
     A=$Particles.Initial.Scale
     B=&VectorFromCurve(
      VectorCurve=DataInterface() {propertyValues {
        XCurve {keys [{arriveTangent=0.5683, interpMode=RCIM_Cubic, leaveTangent=0.5683, tangentMode=RCTM_User, time=0, value=1}, <at 1 0.5932 RCIM_Cubic>]}
        YCurve {keys [
          {arriveTangent=1.5093, interpMode=RCIM_Cubic, leaveTangent=1.5093, tangentMode=RCTM_User, time=0, value=1}
          {arriveTangent=0.5734, tangentMode=RCTM_User, time=1, value=1.5}
         ]}
        ZCurve {keys [{arriveTangent=9.0725, interpMode=RCIM_Cubic, leaveTangent=9.0725, tangentMode=RCTM_User, time=0, value=1}]}
       }}
      Scale Curve=<1 1 1.5>
     )
    )) {isSetParameters=true}
   /SetVariables(Particles.Scale=&Multiply_Vector(
     A=$Particles.Initial.Scale
     B=&VectorFromCurve(
      VectorCurve=DataInterface() {propertyValues {
        XCurve {keys [{arriveTangent=0.6387, interpMode=RCIM_Cubic, leaveTangent=0.6387, tangentMode=RCTM_User, time=0, value=0.5}, <at 0.9989 1>]}
        YCurve {keys [<at 0 0.5>,<at 0.9989 0.75>]}
        ZCurve {keys [<at 0 0>,<at 0.2345 0.7983>,<at 0.9989 1>]}
       }}
      CurveIndex=$Emitter.Scalability.DistanceFraction
     )
    )) {isSetParameters=true}
   /SetVariables(Particles.Scale=&Multiply_Vector(
     A=$Particles.Initial.Scale
     B=&VectorFromCurve(VectorCurve=DataInterface() {propertyValues {
        XCurve {keys [
          {arriveTangent=4.6994, interpMode=RCIM_Cubic, leaveTangent=4.6994, tangentMode=RCTM_User, time=0, value=0.0795}
          {arriveTangent=0.4589, interpMode=RCIM_Cubic, leaveTangent=0.4589, tangentMode=RCTM_User, time=0.15, value=0.7609}
          {arriveTangent=-2.0845, tangentMode=RCTM_User, time=1, value=0.107}
         ]}
        YCurve {keys [
          {arriveTangent=2.5596, interpMode=RCIM_Cubic, leaveTangent=2.5596, tangentMode=RCTM_User, time=0, value=0.0767}
          {arriveTangent=1.1091, interpMode=RCIM_Cubic, leaveTangent=1.1091, time=0.2662, value=0.8916}
          {arriveTangent=-0.1129, interpMode=RCIM_Cubic, leaveTangent=-0.1129, tangentMode=RCTM_User, time=1, value=1.1858}
         ]}
        ZCurve {keys [<at 0 0 RCIM_Cubic>, {arriveTangent=0.5006, interpMode=RCIM_Cubic, leaveTangent=0.5006, time=0.6358, value=1.0328}, <at 0.9988 0.5>]}
       }})
    )) {isSetParameters=true}
   &Drag(Drag=0, Rotational Drag=20, Use Linear Drag=true, Use Rotational Drag=true)
   &ScaleColor(
    Scale Alpha=&MergeHelper_FloatSwitch(
     0=&FloatFromCurve(FloatCurve=DataInterface() {propertyValues {Curve {keys [
          {arriveTangent=9.9926, interpMode=RCIM_Cubic, leaveTangent=9.9927, tangentMode=RCTM_User, time=0, value=0}
          {arriveTangent=-2.8339, interpMode=RCIM_Cubic, leaveTangent=-2.8339, tangentMode=RCTM_User, time=0.0216, value=1}
          <at 1.0011 0.002 RCIM_Cubic>
         ]}}})
     1=&FloatFromCurve(
      FloatCurve=DataInterface() {propertyValues {Curve {keys [
          {arriveTangent=4.2411, interpMode=RCIM_Cubic, leaveTangent=4.2411, tangentMode=RCTM_User, time=0, value=0}
          {arriveTangentWeight=0.0667, interpMode=RCIM_Cubic, leaveTangentWeight=0.1349, tangentMode=RCTM_User, time=0.1, value=1}
          {tangentMode=RCTM_User, time=1, value=0}
         ]}}}
      Scale Curve=10
     )
     2=?
     3=?
     Index=&SpawnGroup()
     Num=::NewEnumerator0
    )
    Scale RGB=&MergeHelper_VectorSwitch(
     0=<1 1 1>
     1=&VectorFromFloat(Value=&FNTimeOfDayScalar(Night Scale=1.5, Morning Scale=1.3, Day Scale=1.5, Evening Scale=1.2))
     2=?
     3=?
     Index=&SpawnGroup()
     Num=::NewEnumerator0
    )
    ScaleA=true
    ScaleRGB=true
   )
   &AccelerationForce(Acceleration=&Lerp_Vector(
     A=<0 0 0>
     B=<0 0 -980>
     Alpha=&FloatFromCurve(FloatCurve=DataInterface() {propertyValues {Curve {keys [{interpMode=RCIM_Cubic, tangentMode=RCTM_User, time=0.5, value=0}, <at 1 1>]}}})
    ))
   &SolveForcesAndVelocity(Clamp Velocity=false, Limit Acceleration=false)
   &DynamicMaterialParameters(
    Index 0 Param 1=&FloatFromCurve(FloatCurve=DataInterface() {propertyValues {Curve {keys [
         {arriveTangent=0.2615, interpMode=RCIM_Cubic, leaveTangent=0.2615, tangentMode=RCTM_User, time=0, value=-1.7984}
         {arriveTangent=3.2109, interpMode=RCIM_Cubic, leaveTangent=3.2109, tangentMode=RCTM_User, time=0.9985, value=0}
        ]}}})
    Index 0 Param 2=&Add_Float(
     A=&FloatFromCurve(FloatCurve=DataInterface() {propertyValues {Curve {keys [
          {arriveTangent=-9.4845, interpMode=RCIM_Cubic, leaveTangent=-9.4845, tangentMode=RCTM_User, time=0, value=4}
          {arriveTangent=0.2383, interpMode=RCIM_Cubic, leaveTangent=0.2383, tangentMode=RCTM_User, time=1, value=0}
         ]}}})
     B=0
    )
    Index 0 Param 3=9
    Index 1 Param 1=&FloatFromCurve(FloatCurve=DataInterface() {propertyValues {Curve {keys [
         <at 0 0.1 RCIM_Cubic>
         {arriveTangent=0.9, interpMode=RCIM_Cubic, leaveTangent=0.9, time=0.5861, value=0.2}
         {arriveTangent=3.3294, interpMode=RCIM_Cubic, leaveTangent=3.3294, tangentMode=RCTM_User, time=1, value=1}
        ]}}})
    Index 1 Param 2=&RandomRangeFloat(Minimum=0, Maximum=1)
    Module.Param0WriteEnabled=true
    Module.Param0WriteEnabled1=true
    Module.Param1WriteEnabled=true
    Module.Param1WriteEnabled1=true
    Module.Param2WriteEnabled=true
    Module.Param2WriteEnabled1=true
    Module.Param3WriteEnabled=true
    Module.Param3WriteEnabled1=true
    Write Parameter Index 1=true
   )
  ]
  rendererClasses [&NiagaraMeshRendererProperties]
  renderers [
   {rendererClass=&NiagaraMeshRendererProperties, rendererIndex=0}
   {rendererClass=&NiagaraMeshRendererProperties, rendererIndex=1}
   {rendererClass=&NiagaraMeshRendererProperties, rendererIndex=2}
   {rendererClass=&NiagaraMeshRendererProperties, rendererIndex=3}
  ]
 }
}
summary {
 systemName=NS_Emote_TideCompost_Board_Splash
 emitters [
  {emitterName=Wake_Mesh, rendererClasses [&NiagaraMeshRendererProperties]}
  {
   emitterName=Edge_Spray_Mesh
   rendererClasses [&NiagaraMeshRendererProperties, &NiagaraSpriteRendererProperties]
  }
  {emitterName=Splash_Both, rendererClasses [&NiagaraSpriteRendererProperties]}
  {emitterName=Ribbons, rendererClasses [&NiagaraMeshRendererProperties]}
 ]
 userVariables [
  User.bIsFrontEnd {type=NiagaraBool}
  User.bIsFrontEndPreview {type=NiagaraBool}
  User.ForwardDot {type=NiagaraFloat, defaultValue=0, description=""}
  User.RightDot {type=NiagaraFloat, defaultValue=0, description=""}
  User.SkeletalMesh {
   type=NiagaraDataInterfaceSkeletalMesh
   defaultValue {
    struct=NiagaraExt_VariableValue_DataInterface
    value {dataInterface=&self:NiagaraDataInterfaceSkeletalMesh_0}
   }
  }
  User.Thruster {type=NiagaraFloat, defaultValue=1, description=""}
 ]
}
