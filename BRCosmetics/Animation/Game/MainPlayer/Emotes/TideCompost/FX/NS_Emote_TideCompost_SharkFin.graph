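# Niagara system for the TideCompost emote that draws a shark fin with wake, splashes and a water trail.
# The Minimal emitter uses spawn groups to split mesh and sprite particles into fin elements following System.SharkPosition/SharkRotation, wake pieces at System.WakePosition, and gravity-driven splashes.
# WaterTrail streams a ribbon from the wake position, with animated width and outline color.
# User.FinHeight drives scale and opacity, while time-of-day functions adjust brightness and color.
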NiagaraSystem /BRCosmetics/Animation/Game/MainPlayer/Emotes/TideCompost/FX/NS_Emote_TideCompost_SharkFin
dependencies {
 usedDataInterfaces [&NiagaraDataInterfaceCurve, &NiagaraDataInterfaceVectorCurve, &NiagaraDataInterfaceColorCurve]
 usedDynamicInputs [&Waveform, &FN_FrontEndFloatModifier]
 usedModules [
  &self:SystemSpawnScript.SystemScriptSource.SystemScriptGraph.NiagaraNodeAssignment_1.SetVariables_#1
  &SystemState
  &BlendSpaceHelper
  &self:SystemSpawnScript.SystemScriptSource.SystemScriptGraph.NiagaraNodeAssignment_0.SetVariables_#2
  &self:SystemSpawnScript.SystemScriptSource.SystemScriptGraph.NiagaraNodeAssignment_2.SetVariables_#3
  &self:SystemSpawnScript.SystemScriptSource.SystemScriptGraph.NiagaraNodeAssignment_3.SetVariables_#4
 ]
}
emitters {
 Minimal {
  emitterUpdateScript [
   &EmitterState()
   &SpawnBurst_Instantaneous(Use Spawn Probability=false, UseLoopCountLimit=false)
   &SpawnRate(SpawnGroup=1, SpawnRate=6, Use Spawn Probability=false)
   &SpawnRate(SpawnGroup=2, SpawnRate=8, Use Spawn Probability=false) {module=SpawnRate002}
   &SpawnBurst_Instantaneous(Spawn Time=0.5667, Use Spawn Probability=false, UseLoopCountLimit=false) {module=SpawnBurst_Instantaneous001}
  ]
  particleSpawnScript [
   /SetVariables(Particles.VisibilityTag=&SpawnGroup()) {isSetParameters=true}
   &PartitionParticles()
   &InitializeParticle(
    Color=&MaskLinearColor(
     Mask Source=0
     Mask Parameter=?
     Mask=2
     Masked Linear Color=&Multiply_LinearColorByFloat(
      Linear Color=&RandomRangeLinearColor(Minimum=<rgba 1.0104 1.4804 2 0.5>, Maximum=<rgba 0.4891 0.8557 1.5 0.7>)
      Float=$User.FinHeight
     )
     Passthrough Linear Color=<rgba 1 1 1 1>
    )
    Lifetime=&MaskFloat(
     Mask Source=0
     Mask Parameter=?
     Mask=0
     Masked Float=5
     Passthrough Float=&RandomRangeFloat(Minimum=0.4, Maximum=0.5)
    )
    Mesh Scale=&MergeHelper_VectorSwitch(
     0=<0 0 0>
     1=&UniformRangedVector(Minimum=<0.4 0.01 0.16>, Maximum=<0.6 0.03 0.23>)
     2=&VectorFromFloat(Value=$User.FinHeight)
     3=?
     Index=$Emitter.SpawnGroup
     Num=::NewEnumerator1
    )
    Mesh Scale Mode=::NewEnumerator4
    Sprite Size Mode=::NewEnumerator3
    Sprite UV Mode=::NewEnumerator3
    Uniform Sprite Size=&RandomRangeFloat(Minimum=30, Maximum=50)
    UsePositionOffset=false
    Write Material Random=true
   )
   &CameraOffset(Camera Offset Amount=&MaskFloat(Mask Source=0, Mask Parameter=?, Mask=0, Masked Float=0, Passthrough Float=5))
   &AddVelocity(
    Velocity=&RandomRangeVector(Minimum=<-50 -150 100>, Maximum=<50 -250 150>)
    Velocity Speed Scale=&MaskFloat(Mask Source=0, Mask Parameter=?, Mask=2, Masked Float=1, Passthrough Float=0)
   )
   &ShapeLocation(MaskSpawnGroup=false, Override Seed=false, Shape Origin=$System.SharkPosition, Sphere Radius=0, UseOwnerScale=true)
  ]
  particleUpdateScript [
   &ParticleState()
   &UpdateMeshOrientation(
    Facing Direction=&MergeHelper_VectorSwitch(0=$System.SharkRotation, 1=<1 0 0>, 2=<1 0 0>, 3=?, Index=&SpawnGroup(), Num=::NewEnumerator1)
    Facing Interpolation=false
    Orientation Method=::NewEnumerator1
    Override Quat=false
    TransformFacing=false
    TransformUp=false
   )
   &DynamicMaterialParameters(
    Index 0 Param 1=&FloatFromCurve(FloatCurve=DataInterface() {propertyValues {Curve {keys [
         {arriveTangent=0.2615, interpMode=RCIM_Cubic, leaveTangent=0.2615, tangentMode=RCTM_User, time=0, value=-1.7984}
         {arriveTangent=3.2109, interpMode=RCIM_Cubic, leaveTangent=3.2109, tangentMode=RCTM_User, time=0.9985, value=0}
        ]}}})
    Index 0 Param 2=&Add_Float(
     A=&FloatFromCurve(FloatCurve=DataInterface() {propertyValues {Curve {keys [
          {arriveTangent=-9.4845, interpMode=RCIM_Cubic, leaveTangent=-9.4845, tangentMode=RCTM_User, time=0, value=4}
          {arriveTangent=0.2383, interpMode=RCIM_Cubic, leaveTangent=0.2383, tangentMode=RCTM_User, time=1, value=0}
         ]}}})
     B=0
    )
    Index 1 Param 1=&FloatFromCurve(FloatCurve=DataInterface() {propertyValues {Curve {keys [
         <at 0 0.1 RCIM_Cubic>
         {arriveTangent=0.9, interpMode=RCIM_Cubic, leaveTangent=0.9, time=0.5861, value=0.2}
         {arriveTangent=3.3294, interpMode=RCIM_Cubic, leaveTangent=3.3294, tangentMode=RCTM_User, time=1, value=1}
        ]}}})
    Index 1 Param 2=&RandomRangeFloat(Minimum=0, Maximum=1)
    Module.Param0WriteEnabled=true
    Module.Param0WriteEnabled1=true
    Module.Param1WriteEnabled=true
    Module.Param1WriteEnabled1=true
    Module.Param2WriteEnabled=false
    Module.Param2WriteEnabled1=true
    Module.Param3WriteEnabled=true
    Module.Param3WriteEnabled1=true
    Write Parameter Index 1=true
   )
   /SetVariables(Particles.Position=&MergeHelper_VectorSwitch(
     0=Linked() {linkedVariable {name=System.SharkPosition, type=NiagaraPosition}}
     1=Linked() {linkedVariable {name=System.WakePosition, type=NiagaraPosition}}
     2=Linked() {linkedVariable {name=Particles.Position, type=NiagaraPosition}}
     3=?
     Index=&SpawnGroup()
     Num=::NewEnumerator1
    )) {isSetParameters=true}
   /SetVariables(
    Particles.Scale=&MergeHelper_VectorSwitch(
     0=<1 1 1>
     1=&Multiply_Vector(
      A=$Particles.Initial.Scale
      B=&VectorFromCurve(
       VectorCurve=DataInterface() {propertyValues {
         XCurve {keys [
           {arriveTangent=4.6994, interpMode=RCIM_Cubic, leaveTangent=4.6994, tangentMode=RCTM_User, time=0, value=0.0795}
           {arriveTangent=0.4589, interpMode=RCIM_Cubic, leaveTangent=0.4589, tangentMode=RCTM_User, time=0.15, value=0.7609}
           {arriveTangent=-2.0845, tangentMode=RCTM_User, time=1, value=0.107}
          ]}
         YCurve {keys [
           {arriveTangent=0.9157, interpMode=RCIM_Cubic, leaveTangent=0.9157, tangentMode=RCTM_User, time=0, value=0.1206}
           {arriveTangent=-0.0038, tangentMode=RCTM_User, time=1, value=0.2802}
          ]}
         ZCurve {keys [<at 0 0 RCIM_Cubic>, {arriveTangent=0.5006, interpMode=RCIM_Cubic, leaveTangent=0.5006, time=0.33, value=1}, <at 0.9988 0.5>]}
        }}
       Scale Curve=&MakeVector(
        X=&FloatFromCurve(
         FloatCurve=DataInterface() {propertyValues {Curve {keys [{interpMode=RCIM_Cubic, tangentMode=RCTM_User, time=0, value=0.3}, <at 0.3 1>]}}}
         CurveIndex=$User.FinHeight
        )
        Y=&FloatFromCurve(
         FloatCurve=DataInterface() {propertyValues {Curve {keys [{interpMode=RCIM_Cubic, tangentMode=RCTM_User, time=0, value=0.5}, <at 1 1>]}}}
         CurveIndex=$User.FinHeight
        )
        Z=&FloatFromCurve(
         FloatCurve=DataInterface() {propertyValues {Curve {keys [
             {interpMode=RCIM_Cubic, tangentMode=RCTM_User, time=0, value=0}
             {arriveTangent=1, interpMode=RCIM_Cubic, leaveTangent=1, time=0.0518, value=0.1}
             <at 1 1>
            ]}}}
         CurveIndex=$User.FinHeight
        )
       )
      )
     )
     2=&VectorFromCurve(VectorCurve=DataInterface() {propertyValues {
        XCurve {keys [
          {arriveTangent=1.1339, interpMode=RCIM_Cubic, leaveTangent=1.1339, tangentMode=RCTM_User, time=0, value=0.3}
          {arriveTangent=0.6037, tangentMode=RCTM_User, time=1, value=1}
         ]}
        YCurve {keys [{arriveTangent=2.2304, interpMode=RCIM_Cubic, leaveTangent=2.2304, tangentMode=RCTM_User, time=0, value=0}, <at 1 1>]}
        ZCurve {keys [{arriveTangent=2.2304, interpMode=RCIM_Cubic, leaveTangent=2.2304, tangentMode=RCTM_User, time=0, value=0}, <at 1 1>]}
       }})
     3=?
     Index=&SpawnGroup()
     Num=::NewEnumerator1
    )
    Particles.SpriteSize=&Multiply_Vector2D_ByFloat(
     Float=&FloatFromCurve(
      FloatCurve=DataInterface() {propertyValues {Curve {
         defaultValue=1
         keys [
          {arriveTangent=1.7309, interpMode=RCIM_Cubic, leaveTangent=1.7309, tangentMode=RCTM_User, time=0, value=0}
          {tangentMode=RCTM_User, time=1, value=1}
         ]
        }}}
      Scale Curve=20
     )
     Vector2D=<1 3>
    )
   ) {isSetParameters=true}
   /SetVariables(Particles.Splash DynP 00:Vector4f=&MakeVector4(
     X=&FloatFromCurve(FloatCurve=DataInterface() {propertyValues {Curve {keys [{interpMode=RCIM_Cubic, tangentMode=RCTM_Break, time=0.6, value=1}, {interpMode=RCIM_Cubic, tangentMode=RCTM_Break, time=1, value=0.4}]}}})
     Y=0
     Z=0
     W=0
    )) {isSetParameters=true}
   &InterpolateOverTime(
    Target Vector Value:Vector3f=&TransformPosition(Position=$Particles.Position, Source Space=::Local, Destination Space=::World)
    Variable Type=::NewEnumerator2
   )
   /SetVariables(
    Particles.Ribbon Color:LinearColor=&MakeLinearColorFromVectorAndFloat(
     "Vector (RGB)"=&MakeVectorFromLinearColorRGB(Color=&Multiply_LinearColorByFloat(
       Linear Color=<rgba 0.4891 0.8557 1.5 0.7>
       Float=&FNTimeOfDayScalar(Night Scale=0.3, Morning Scale=1, Evening Scale=0.6)
      ))
     "Float (Alpha)"=&Lerp_Float(A=0, B=1, Alpha=$User.FinHeight)
    )
    Particles.Ribbon Dynp 00:Vector4f=&MakeVector4(X=&RandomRangeFloat(Minimum=0, Maximum=1), Y=0, Z=0, W=0)
    Particles.Ribbon Dynp 01:Vector4f=&MakeVector4(
     X=&MakeFloatFromVector(Vector=$StackContext.InterpolateOverTime.InterpolateOverTime_V3_.Moving Average, Channel=::NewEnumerator0)
     Y=&MakeFloatFromVector(Vector=$StackContext.InterpolateOverTime.InterpolateOverTime_V3_.Moving Average, Channel=::NewEnumerator1)
     Z=&Add_Float(
      A=&MakeFloatFromVector(Vector=$StackContext.InterpolateOverTime.InterpolateOverTime_V3_.Moving Average, Channel=::NewEnumerator2)
      B=&FloatFromCurve(FloatCurve=DataInterface() {propertyValues {Curve {
          defaultValue=1
          keys [
           {arriveTangent=-0.0224, interpMode=RCIM_Cubic, leaveTangent=-0.0224, leaveTangentWeight=0.2653, tangentMode=RCTM_User, time=0, value=0}
           {
            arriveTangent=-1.1179
            arriveTangentWeight=0.3979
            interpMode=RCIM_Cubic
            leaveTangent=-1.1179
            leaveTangentWeight=0.1021
            tangentMode=RCTM_User
            time=0.7958
            value=-0.3772
           }
           {arriveTangent=-4.1785, arriveTangentWeight=0.2925, interpMode=RCIM_Cubic, leaveTangent=-4.1785, tangentMode=RCTM_User, time=1, value=-1}
          ]
         }}})
     )
     W=0
    )
   ) {isSetParameters=true}
   &ScaleColor(
    Scale Alpha=&MergeHelper_FloatSwitch(
     0=&FloatFromCurve(
      FloatCurve=DataInterface() {propertyValues {Curve {keys [
          {arriveTangent=6.9881, interpMode=RCIM_Cubic, leaveTangent=6.9881, tangentMode=RCTM_User, time=0, value=0}
          {arriveTangent=5.1399, interpMode=RCIM_Cubic, leaveTangent=5.1399, time=0.028, value=0.365}
          {arriveTangent=1.2299, interpMode=RCIM_Cubic, leaveTangent=1.2299, tangentMode=RCTM_User, time=0.1, value=0.514}
          {arriveTangent=0.5001, interpMode=RCIM_Cubic, leaveTangent=0.5001, time=0.4583, value=0.7269}
          {arriveTangent=0.5041, interpMode=RCIM_Cubic, leaveTangent=0.5041, time=0.8962, value=0.9122}
          {arriveTangent=0.3095, interpMode=RCIM_Cubic, leaveTangent=0.3095, tangentMode=RCTM_User, time=1, value=1}
         ]}}}
      CurveIndex=$User.FinHeight
     )
     1=&FloatFromCurve(
      FloatCurve=DataInterface() {propertyValues {Curve {keys [
          {arriveTangent=12.675, interpMode=RCIM_Cubic, leaveTangent=12.675, tangentMode=RCTM_User, time=0, value=0}
          {interpMode=RCIM_Cubic, tangentMode=RCTM_User, time=0.0633, value=1}
         ]}}}
      Scale Curve=0.7
     )
     2=&FloatFromCurve(
      FloatCurve=DataInterface() {propertyValues {Curve {
         defaultValue=1
         keys [
          {arriveTangent=1.6865, interpMode=RCIM_Cubic, leaveTangent=1.6865, tangentMode=RCTM_User, time=0, value=0}
          {interpMode=RCIM_Cubic, tangentMode=RCTM_User, time=0.1, value=10}
          <at 1 0>
         ]
        }}}
      Scale Curve=0.5
     )
     3=?
     Index=&SpawnGroup()
     Num=::NewEnumerator1
    )
    Scale RGB=&Multiply_VectorByFloat(
     Vector=&VectorFromFloat(Value=&FNTimeOfDayScalar(Night Scale=0.5, Morning Scale=1, Evening Scale=0.7))
     Float=&MergeHelper_FloatSwitch(0=1, 1=1, 2=?, 3=?, Index=&SpawnGroup(), Num=::NewEnumerator0)
    )
    ScaleA=true
    ScaleRGB=true
   )
   &GravityForce(Gravity=&MaskVector(
     Mask Source=0
     Mask Parameter=?
     Mask=2
     Masked Vector=&RandomRangeVector(Minimum=<0 0 -500>, Maximum=<0 0 -700>)
     Passthrough Vector=<0 0 0>
    ))
   &Drag(Drag=2, Use Linear Drag=true, Use Rotational Drag=false)
   &SolveForcesAndVelocity(Clamp Velocity=false, Limit Acceleration=false)
  ]
  rendererClasses [&NiagaraMeshRendererProperties, &NiagaraSpriteRendererProperties]
  renderers [
   {rendererClass=&NiagaraMeshRendererProperties, rendererIndex=0}
   {rendererClass=&NiagaraMeshRendererProperties, rendererIndex=1}
   {rendererClass=&NiagaraMeshRendererProperties, rendererIndex=2}
   {rendererClass=&NiagaraMeshRendererProperties, rendererIndex=3}
   {rendererClass=&NiagaraMeshRendererProperties, rendererIndex=4}
   {rendererClass=&NiagaraMeshRendererProperties, rendererIndex=5}
   {rendererClass=&NiagaraSpriteRendererProperties, rendererIndex=6}
  ]
 }
 WaterTrail {
  emitterUpdateScript [
   &EmitterState()
   &SpawnRate(SpawnRate=50, Use Spawn Probability=false)
  ]
  particleSpawnScript [
   &InitializeParticle(
    Color=<rgba 0.6 1.2021 2 1>
    Lifetime=0.3
    Ribbon Facing Mode=::NewEnumerator1
    Ribbon Facing Vector:Vector3f=$Engine.Owner.SystemZAxis
    Ribbon Width=50
    Ribbon Width Mode=::NewEnumerator1
    UsePositionOffset=false
    Write Material Random=true
   )
   &AddVelocity(Velocity=&MakeVector(X=0, Y=-1000, Z=0))
   /SetVariables(Particles.Position=&Add_Vector(
     A=Linked() {linkedVariable {name=System.WakePosition, type=NiagaraPosition}}
     B=&Multiply_Vector(A=Linked() {linkedVariable {name=StackContext.Position, type=NiagaraPosition}}, B=<1 1 0>)
    )) {isSetParameters=true}
  ]
  particleUpdateScript [
   &ParticleState()
   /SetVariables(Particles.Position=&Lerp_Vector(
     A=Linked() {linkedVariable {name=System.WakePosition, type=NiagaraPosition}}
     B=Linked() {linkedVariable {name=Particles.Position, type=NiagaraPosition}}
     Alpha=&FloatFromCurve(FloatCurve=DataInterface() {propertyValues {Curve {keys [
          {
           arriveTangent=0.3465
           arriveTangentWeight=0.1684
           interpMode=RCIM_Cubic
           leaveTangent=0.3465
           leaveTangentWeight=0.1684
           tangentMode=RCTM_User
           tangentWeightMode=RCTWM_WeightedBoth
           time=0
           value=0
          }
          {arriveTangent=-0.0264, tangentMode=RCTM_User, time=0.3, value=1}
         ]}}})
    )) {isSetParameters=true}
   &ScaleColor(
    Scale Alpha=&FloatFromCurve(
     FloatCurve=DataInterface() {propertyValues {Curve {keys [
         {interpMode=RCIM_Cubic, tangentMode=RCTM_Break, time=0, value=1}
         {arriveTangent=-0.4167, interpMode=RCIM_Cubic, leaveTangent=-0.4167, time=0.15, value=1}
         {arriveTangent=-1.5385, interpMode=RCIM_Cubic, leaveTangent=-1.5385, time=0.4871, value=0.797}
         {interpMode=RCIM_Cubic, tangentMode=RCTM_Break, time=0.8, value=0}
        ]}}}
     Scale Curve=&Lerp_Float(A=0, B=0.5, Alpha=$User.FinHeight)
    )
    Scale RGB=&VectorFromFloat(Value=&FNTimeOfDayScalar(Night Scale=0.5, Morning Scale=1, Evening Scale=0.7))
    ScaleA=true
    ScaleRGB=true
   )
   &ScaleRibbonWidth(Ribbon Width Scale=&FloatFromCurve(FloatCurve=DataInterface() {propertyValues {Curve {keys [
         {arriveTangent=4.0157, interpMode=RCIM_Cubic, leaveTangent=4.0157, tangentMode=RCTM_User, time=0.0009, value=0}
         {arriveTangent=0.0892, interpMode=RCIM_Cubic, leaveTangent=0.0892, tangentMode=RCTM_User, time=0.4162, value=1.5758}
         <at 0.9976 0 RCIM_Cubic>
        ]}}}))
   &Drag(Drag=3, Use Linear Drag=true, Use Rotational Drag=true)
   &GravityForce(Gravity=<-500 0 0>)
   &SolveForcesAndVelocity(Clamp Velocity=false, Limit Acceleration=false)
   &BlendSpace(Rotation Contribution=0)
   /SetVariables(Particles.Wake Outline Width:NiagaraFloat=&Multiply_Float(
     A=&Add_Float(A=50, B=0)
     B=&FloatFromCurve(FloatCurve=DataInterface() {propertyValues {Curve {
         defaultValue=1
         keys [
          <at -0.0018 0.2235 RCIM_Cubic>
          {arriveTangent=1.9234, interpMode=RCIM_Cubic, leaveTangent=1.9234, time=0.1213, value=0.5835}
          {arriveTangent=0.8499, interpMode=RCIM_Cubic, leaveTangent=0.8499, time=0.2083, value=0.6275}
          {arriveTangent=0.1196, interpMode=RCIM_Cubic, leaveTangent=0.1196, time=0.2775, value=0.7163}
          {arriveTangent=0.4531, interpMode=RCIM_Cubic, leaveTangent=0.4531, time=0.4624, value=0.6579}
          {arriveTangent=0.5513, interpMode=RCIM_Cubic, leaveTangent=0.5513, time=0.6262, value=0.8743}
          {arriveTangent=0.3363, interpMode=RCIM_Cubic, leaveTangent=0.3363, time=0.8088, value=0.8489}
          {arriveTangent=1.3622, interpMode=RCIM_Cubic, leaveTangent=1.3622, tangentMode=RCTM_User, time=1, value=1}
         ]
        }}})
    )) {isSetParameters=true}
   /SetVariables(Particles.Wake Outline Color:LinearColor=&MakeLinearColorFromVectorAndFloat(
     "Vector (RGB)"=&MakeVectorFromLinearColorRGB(Color=&FNTimeOfDayColor(
       Night Color=&ColorFromCurve(ColorCurve=DataInterface() {propertyValues {
          AlphaCurve {keys [<at 0.7797 1>,<at 1 0>]}
          BlueCurve {keys [
            <at 0.1179 1>
            <at 0.1781 1.35>
            <at 0.2855 1.15>
            {arriveTangent=-0.2113, interpMode=RCIM_Cubic, leaveTangent=-0.2113, time=0.3872, value=1}
            <at 0.9955 1>
           ]}
          GreenCurve {keys [
            <at 0.1179 0.5486>
            <at 0.1781 1.1736>
            <at 0.2855 0.9997>
            {arriveTangent=-0.8299, interpMode=RCIM_Cubic, leaveTangent=-0.8299, time=0.3872, value=0.8258}
            <at 0.9955 0.4105>
           ]}
          RedCurve {keys [
            <at 0.1179 0>
            <at 0.1781 1.1391>
            <at 0.2855 0.9704>
            {arriveTangent=-1.3667, interpMode=RCIM_Cubic, leaveTangent=-1.3667, time=0.3872, value=0.3544}
            <at 0.9955 0>
           ]}
         }})
       Morning Color=&ColorFromCurve(ColorCurve=DataInterface() {propertyValues {
          AlphaCurve {keys [<at 0.7797 1>,<at 1 0>]}
          BlueCurve {keys [
            <at 0.1179 1>
            <at 0.1781 1.5>
            <at 0.2855 1.3>
            {arriveTangent=-0.4225, interpMode=RCIM_Cubic, leaveTangent=-0.4225, time=0.3872, value=1}
            <at 0.9955 1>
           ]}
          GreenCurve {keys [
            <at 0.1179 0.5486>
            <at 0.1781 1.304>
            <at 0.2855 1.1301>
            {arriveTangent=-1.0136, interpMode=RCIM_Cubic, leaveTangent=-1.0136, time=0.3872, value=0.8258}
            <at 0.9955 0.4105>
           ]}
          RedCurve {keys [
            <at 0.1179 0>
            <at 0.1781 1.2657>
            <at 0.2855 1.0969>
            {arriveTangent=-1.545, interpMode=RCIM_Cubic, leaveTangent=-1.545, time=0.3872, value=0.3544}
            <at 0.9955 0>
           ]}
         }})
       Day Color=&ColorFromCurve(ColorCurve=DataInterface() {propertyValues {
          AlphaCurve {keys [<at 0.7797 1>,<at 1 0>]}
          BlueCurve {keys [
            <at 0.1179 1>
            <at 0.1781 1.5>
            <at 0.2855 1.3>
            {arriveTangent=-0.4225, interpMode=RCIM_Cubic, leaveTangent=-0.4225, time=0.3872, value=1}
            <at 0.9955 1>
           ]}
          GreenCurve {keys [
            <at 0.1179 0.5486>
            <at 0.1781 1.304>
            <at 0.2855 1.1301>
            {arriveTangent=-1.0136, interpMode=RCIM_Cubic, leaveTangent=-1.0136, time=0.3872, value=0.8258}
            <at 0.9955 0.4105>
           ]}
          RedCurve {keys [
            <at 0.1179 0>
            <at 0.1781 1.2657>
            <at 0.2855 1.0969>
            {arriveTangent=-1.545, interpMode=RCIM_Cubic, leaveTangent=-1.545, time=0.3872, value=0.3544}
            <at 0.9955 0>
           ]}
         }})
       Evening Color=&ColorFromCurve(ColorCurve=DataInterface() {propertyValues {
          AlphaCurve {keys [<at 0.7797 1>,<at 1 0>]}
          BlueCurve {keys [
            <at 0.1179 1>
            <at 0.1781 1.5>
            <at 0.2855 1.3>
            {arriveTangent=-0.4225, interpMode=RCIM_Cubic, leaveTangent=-0.4225, time=0.3872, value=1}
            <at 0.9955 1>
           ]}
          GreenCurve {keys [
            <at 0.1179 0.5486>
            <at 0.1781 1.304>
            <at 0.2855 1.1301>
            {arriveTangent=-1.0136, interpMode=RCIM_Cubic, leaveTangent=-1.0136, time=0.3872, value=0.8258}
            <at 0.9955 0.4105>
           ]}
          RedCurve {keys [
            <at 0.1179 0>
            <at 0.1781 1.2657>
            <at 0.2855 1.0969>
            {arriveTangent=-1.545, interpMode=RCIM_Cubic, leaveTangent=-1.545, time=0.3872, value=0.3544}
            <at 0.9955 0>
           ]}
         }})
      ))
     "Float (Alpha)"=&FloatFromCurve(
      FloatCurve=DataInterface() {propertyValues {Curve {keys [
          {interpMode=RCIM_Cubic, tangentMode=RCTM_Break, time=0, value=0}
          {arriveTangent=1.6363, interpMode=RCIM_Cubic, leaveTangent=1.6363, time=0.15, value=1}
          {arriveTangent=-1.5385, interpMode=RCIM_Cubic, leaveTangent=-1.5385, time=0.4871, value=0.797}
          {interpMode=RCIM_Cubic, tangentMode=RCTM_Break, time=0.8, value=0}
         ]}}}
      Scale Curve=0.5
     )
    )) {isSetParameters=true}
  ]
  rendererClasses [&NiagaraRibbonRendererProperties]
  renderers [{rendererClass=&NiagaraRibbonRendererProperties}]
 }
}
summary {
 systemName=NS_Emote_TideCompost_SharkFin
 emitters [
  {emitterName=Minimal, rendererClasses [&NiagaraMeshRendererProperties, &NiagaraSpriteRendererProperties]}
  {emitterName=WaterTrail, rendererClasses [&NiagaraRibbonRendererProperties]}
 ]
 userVariables [
  User.bIsFrontEnd {type=NiagaraBool}
  User.bIsFrontEndPreview {type=NiagaraBool}
  User.FinHeight {type=NiagaraFloat, defaultValue=1, description=""}
 ]
}
