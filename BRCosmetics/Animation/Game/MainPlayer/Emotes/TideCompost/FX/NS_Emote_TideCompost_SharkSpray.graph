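# Niagara system for the shark spray in the TideCompost emote, emitted along a socket's position and orientation.
# One "drops" emitter runs three spawn groups, tagged for visibility across two sprite renderers and a mesh renderer.
# Two groups scale spawn rate with User.Spray; the third stays pinned to the socket and takes alpha from it.
# A smoothed system value scales launch speed and opacity, while particles inherit socket movement under drag and acceleration.
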NiagaraSystem /BRCosmetics/Animation/Game/MainPlayer/Emotes/TideCompost/FX/NS_Emote_TideCompost_SharkSpray
dependencies {
 usedDataInterfaces [
  &NiagaraDataInterfaceSpriteRendererInfo
  &NiagaraDataInterfaceCurve
  &NiagaraDataInterfaceVectorCurve
  &NiagaraDataInterfaceSocketReader
  &NiagaraDataInterfaceSkeletalMesh
 ]
 usedDynamicInputs [&self:SpawnGrp_0_1_Mask_06]
 usedModules [
  &self:SystemSpawnScript.SystemScriptSource.SystemScriptGraph.NiagaraNodeAssignment_1.SetVariables_#1
  &SystemState
  &BlendSpaceHelper
  &SampleSkeletalMesh
  &self:SystemSpawnScript.SystemScriptSource.SystemScriptGraph.NiagaraNodeAssignment_0.SetVariables_#2
  &self:SystemSpawnScript.SystemScriptSource.SystemScriptGraph.NiagaraNodeAssignment_3.SetVariables_#3
  &InterpolateOverTime
 ]
}
emitters {drops {
  emitterUpdateScript [
   &EmitterState()
   &SpawnRate(SpawnRate=&Multiply_Float(A=13, B=$User.Spray), Use Spawn Probability=false)
   &SpawnRate(SpawnGroup=1, SpawnRate=&Multiply_Float(A=15, B=$User.Spray), Use Spawn Probability=false) {module=SpawnRate001}
   &SpawnRate(SpawnGroup=2, SpawnRate=8, Use Spawn Probability=false) {module=SpawnRate002}
  ]
  particleSpawnScript [
   /SetVariables(Particles.VisibilityTag=&SpawnGroup()) {isSetParameters=true}
   &InitializeParticle(
    Color=$System.SharkSprayColor
    Lifetime=&MergeHelper_FloatSwitch(
     0=&RandomRangeFloat(Minimum=0.5, Maximum=0.6)
     1=&RandomRangeFloat(Minimum=0.2, Maximum=0.4)
     2=?
     3=?
     Index=&SpawnGroup()
     Num=::NewEnumerator0
    )
    Mesh Scale=<0.05 0.4 0.05>
    Mesh Scale Mode=::NewEnumerator4
    Sprite Rotation Angle=&MergeHelper_FloatSwitch(
     0=&RandomRangeFloat(Minimum=0, Maximum=360)
     1=0
     2=?
     3=?
     Index=&SpawnGroup()
     Num=::NewEnumerator0
    )
    Sprite Rotation Mode=::NewEnumerator0
    Sprite Size=&Multiply_Vector2D_ByFloat(
     Float=1.5
     Vector2D=/Niagara/DynamicInputs/Helpers/MaskVector2d(
      Mask Source:NiagaraInt32=0
      Mask Parameter:NiagaraInt32=?
      Mask:NiagaraInt32=0
      Masked Vector2D:Vector2f=&Vector2DFromFloat(Value=&RandomRangeFloat(Minimum=23, Maximum=40))
      Passthrough Vector2D:Vector2f=&RandomRangeVector2D(Minimum=<18 25>, Maximum=<23 35>)
     )
    )
    Sprite Size Mode=::NewEnumerator4
    Sprite UV Mode=::NewEnumerator3
    UsePositionOffset=false
    Write Material Random=true
   )
   &SubUVAnimation(
    End Frame Range Override=3
    "Sprite Renderer "=DataInterface() {propertyValues {SpriteRenderer=&self:drops_0.NiagaraSpriteRendererProperties_1}}
    SubUV Animation Mode:ENiagara_SubUVLookupModeV2=::NewEnumerator2
    UseEndFrame=true
    UseStartFrame=false
   ) {module=SubUVAnimation001}
   /SetVariables(Particles.Position=$System.SpraySocketPosition) {isSetParameters=true}
   &AddVelocity(
    Additional Quaternion Rotation=false
    Rotation Mode=::NewEnumerator3
    Rotation Quaternion:Quat4f=$System.SpraySocketQuat
    Velocity=<0 0 500>
    Velocity Speed Scale=$System.InterpolateOverTime.InterpolateOverTime_Float.Moving Average
   )
  ]
  particleUpdateScript [
   &BlendSpace() {enabled=false}
   &ParticleState(Lifetime:NiagaraFloat=&Lerp_Float(
     A=&FloatFromCurve(FloatCurve=DataInterface() {propertyValues {Curve {defaultValue=1, keys [{tangentMode=RCTM_User, time=0, value=0}, {tangentMode=RCTM_User, time=1, value=0.5}]}}})
     B=1
     Alpha=0.5
    ))
   &ScaleMeshSize(Scale Factor=&VectorFromCurve(VectorCurve=DataInterface() {propertyValues {
       XCurve {keys [<at 0 1>,<at 1 1>]}
       YCurve {keys [
         {arriveTangent=1.766, interpMode=RCIM_Cubic, leaveTangent=1.766, tangentMode=RCTM_User, time=0, value=0.1}
         {interpMode=RCIM_Cubic, tangentMode=RCTM_User, time=0.3, value=1}
         <at 1 0>
        ]}
       ZCurve {keys [<at 0 1>,<at 1 1>]}
      }}))
   &UpdateMeshOrientation(
    Facing Direction=&RotateVectorByQuaternion(Vector To Rotate=<0 0 1>, Quaternion=$System.SpraySocketQuat)
    Facing Interpolation=false
    Orientation Method=::NewEnumerator1
    Override Quat=false
    TransformFacing=false
    TransformUp=false
   )
   &InheritSourceMovement(
    Applied Position Scale=&Multiply_VectorByFloat(
     Vector=<1 1 1>
     Float=&FloatFromCurve(
      FloatCurve=DataInterface() {propertyValues {Curve {keys [{interpMode=RCIM_Cubic, tangentMode=RCTM_User, time=0.4, value=1}, <at 1 0.3>]}}}
      Scale Curve=&Multiply_Float(A=1, B=$User.Spray)
     )
    )
    ApplyForce=false
    ApplyPosition=true
    ApplyPositionFalloff=false
    Source Velocity:Vector3f=$System.SpraySocketVelocity
   )
   &SpriteRotationRate()
   &ScaleSpriteSize(
    Scale Sprite Size Mode=::NewEnumerator0
    Uniform Scale Factor=&MergeHelper_FloatSwitch(
     0=&FloatFromCurve(FloatCurve=DataInterface() {propertyValues {Curve {keys [
          {interpMode=RCIM_Cubic, leaveTangent=1.046, leaveTangentWeight=0.3333, tangentMode=RCTM_Break, time=0, value=0.3}
          {
           arriveTangent=1.6841
           arriveTangentWeight=0.1
           interpMode=RCIM_Cubic
           leaveTangent=1.6841
           leaveTangentWeight=0.2333
           tangentMode=RCTM_User
           time=0.3267
           value=0.8755
          }
          <at 1 1.5>
         ]}}})
     1=&FloatFromCurve(FloatCurve=DataInterface() {propertyValues {Curve {
         defaultValue=1
         keys [
          {arriveTangent=0.534, interpMode=RCIM_Cubic, leaveTangent=0.534, tangentMode=RCTM_User, time=0, value=0.05}
          {tangentMode=RCTM_User, time=1, value=1}
         ]
        }}})
     2=?
     3=?
     Index=&SpawnGroup()
     Num=::NewEnumerator0
    )
   ) {module=ScaleSpriteSize001}
   &ScaleColor(
    Scale Alpha=&Multiply_Float(
     A=&MergeHelper_FloatSwitch(
      0=&FloatFromCurve(
       FloatCurve=DataInterface() {propertyValues {Curve {keys [
           {interpMode=RCIM_Cubic, tangentMode=RCTM_User, time=0, value=0}
           {arriveTangent=-0.0893, interpMode=RCIM_Cubic, leaveTangent=-0.0893, tangentMode=RCTM_User, time=0.3, value=0.85}
           {arriveTangent=-0.0541, interpMode=RCIM_Cubic, leaveTangent=-0.0541, tangentMode=RCTM_User, time=1, value=0.5}
          ]}}}
       Scale Curve=&FNTimeOfDayScalar(Night Scale=0.85, Morning Scale=0.45, Day Scale=0.25, Evening Scale=0.45)
      )
      1=&FloatFromCurve(FloatCurve=DataInterface() {propertyValues {Curve {keys [
           <at 0 0 RCIM_Cubic>
           {arriveTangent=1.6041, leaveTangent=1.6041, time=0.05, value=1.0113}
           {arriveTangent=-1.3484, interpMode=RCIM_Cubic, leaveTangent=-1.3484, time=0.5, value=1}
           <at 0.8 0 RCIM_Cubic>
          ]}}})
      2=$User.Spray
      3=?
      Index=&SpawnGroup()
      Num=::NewEnumerator1
     )
     B=$System.InterpolateOverTime.InterpolateOverTime_Float.Moving Average
    )
    Scale RGB=&MergeHelper_VectorSwitch(0=<2 2 2>, 1=<1 1 1>, 2=?, 3=?, Index=&SpawnGroup(), Num=::NewEnumerator0)
    ScaleA=true
    ScaleRGB=true
   ) {module=ScaleColor001}
   &Drag(Drag=3, Use Linear Drag=true, Use Rotational Drag=false)
   &AccelerationForce(Acceleration=<0 -700 0>, Coordinate Space=::Local)
   &SolveForcesAndVelocity(Clamp Velocity=false, Limit Acceleration=false)
   &ScaleSpriteSizeBySpeed(Max Scale Factor=&MaskVector2d(
     Mask Source:NiagaraInt32=0
     Mask Parameter:NiagaraInt32=?
     Mask:NiagaraInt32=1
     Masked Vector2D:Vector2f=<2 11>
     Passthrough Vector2D:Vector2f=<1 1>
    ))
   &DynamicMaterialParameters(
    Index 0 Param 1=0.5
    Index 0 Param 2=&FloatFromCurve(FloatCurve=DataInterface() {propertyValues {Curve {keys [
         {interpMode=RCIM_Cubic, tangentMode=RCTM_Break, time=0, value=1}
         {arriveTangent=-2.6912, interpMode=RCIM_Cubic, leaveTangent=-2.0957, tangentMode=RCTM_Break, time=0.7, value=0.3}
         <at 1 0>
        ]}}})
    Index 0 Param 3=&FloatFromCurve(FloatCurve=DataInterface() {propertyValues {Curve {keys [
         {arriveTangent=-2.3777, interpMode=RCIM_Cubic, tangentMode=RCTM_Break, time=0, value=1.35}
         {arriveTangent=0.0149, interpMode=RCIM_Cubic, tangentMode=RCTM_Break, time=1, value=0.4}
        ]}}})
    Module.Index0Float0Write=true
    Module.Index0Float1Write=true
    Module.Index0Float2Write=true
    Module.Index0Float3Write=true
   )
   &SubUVAnimation(
    Playback Scale Curve=DataInterface() {propertyValues {Curve {keys [
        {interpMode=RCIM_Cubic, tangentMode=RCTM_User, time=0.25, value=1}
        {arriveTangent=-1.2656, interpMode=RCIM_Cubic, leaveTangent=-1.2656, tangentMode=RCTM_User, time=1, value=0}
       ]}}}
    Random Start Frame=true
    SubUV Animation Mode:ENiagara_SubUVLookupModeV2=::NewEnumerator3
    UseEndFrame=false
    UsePlayRateScale=true
    UseStartFrame=false
   )
   /SetVariables(Particles.Position=&MergeHelper_VectorSwitch(
     0=Linked() {linkedVariable {name=Particles.Position, type=NiagaraPosition}}
     1=Linked() {linkedVariable {name=Particles.Position, type=NiagaraPosition}}
     2=Linked() {linkedVariable {name=System.SpraySocketPosition, type=NiagaraPosition}}
     3=?
     Index=&SpawnGroup()
     Num=::NewEnumerator1
    )) {isSetParameters=true}
   /SetVariables(Particles.MeshOrientation=$System.SpraySocketQuat) {isSetParameters=true}
  ]
  rendererClasses [&NiagaraSpriteRendererProperties, &NiagaraMeshRendererProperties]
  renderers [
   {rendererClass=&NiagaraSpriteRendererProperties}
   {rendererClass=&NiagaraSpriteRendererProperties, rendererIndex=1}
   {rendererClass=&NiagaraMeshRendererProperties, rendererIndex=2}
  ]
 }}
summary {
 systemName=NS_Emote_TideCompost_SharkSpray
 emitters [{emitterName=drops, rendererClasses [&NiagaraSpriteRendererProperties, &NiagaraMeshRendererProperties]}]
 userVariables [
  User.bIsFrontEnd {type=NiagaraBool}
  User.bIsFrontEndPreview {type=NiagaraBool}
  User.ParaGlideLeanAlpha {type=NiagaraFloat, defaultValue=0, description=""}
  User.Spray {type=NiagaraFloat, defaultValue=1, description=""}
  User.Vel {type=NiagaraFloat, defaultValue=0, description=""}
 ]
}
