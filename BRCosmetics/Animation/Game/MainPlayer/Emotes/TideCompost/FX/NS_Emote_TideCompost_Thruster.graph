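# Niagara system for the rocket thruster flame in the TideCompost emote.
# A single emitter, Missile_FX, spawns particles in groups by burst and rate, feeding four mesh renderers and a sprite renderer for layered flame and flare.
# User.Thruster drives flame size, offset, brightness and flare strength, while dense curves add flicker.
# It also writes flickering particle light attributes scaled by time of day, though no light renderer is listed.
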NiagaraSystem /BRCosmetics/Animation/Game/MainPlayer/Emotes/TideCompost/FX/NS_Emote_TideCompost_Thruster
dependencies {
 usedDataInterfaces [
  &NiagaraDataInterfaceCamera
  &NiagaraDataInterfaceCurve
  &NiagaraDataInterfaceColorCurve
  &NiagaraDataInterfaceSkeletalMesh
 ]
 usedDynamicInputs []
 usedModules [
  &SystemState
  &self:SystemSpawnScript.SystemScriptSource.SystemScriptGraph.NiagaraNodeAssignment_0.SetVariables_#1
 ]
}
emitters {Missile_FX {
  emitterUpdateScript [
   &EmitterState()
   &SpawnBurst_Instantaneous(Spawn Group=2, Use Spawn Probability=false, UseLoopCountLimit=false) {module=SpawnBurst_Instantaneous002}
   &SpawnBurst_Instantaneous(Spawn Count=1, Use Spawn Probability=false)
   &SpawnRate(SpawnRate=3, Use Spawn Probability=false)
   &SpawnRate(SpawnGroup=2, SpawnRate=3, Use Spawn Probability=false) {module=SpawnRate001}
   /SetVariables(
    Emitter.Flame1A:LinearColor=<rgba 0.125 0.0688 0.0125 1>
    Emitter.Flame1B:LinearColor=<rgba 0.5 0.4333 0.1 1>
    Emitter.Flame2A:LinearColor=<rgba 1 0.6667 0.2 1>
    Emitter.Flame2B:LinearColor=<rgba 0.4 0.2133 0.08 1>
    Emitter.Flame3A:LinearColor=<rgba 1 0.4838 0.1 1>
    Emitter.Flame3B:LinearColor=<rgba 0.4583 0.2796 0.0458 1>
    Emitter.FlameCoreA:LinearColor=<rgba 1 0.9088 0.526 1>
    Emitter.FlameCoreB:LinearColor=<rgba 1 0.7396 0.2188 1>
    Emitter.FlameDark:LinearColor=<rgba 0.1927 0.1172 0.0416 1>
    Emitter.FlameLight:LinearColor=<rgba 0.3854 0.2318 0.0781 1>
   ) {isSetParameters=true}
   /SetVariables(Emitter.float:NiagaraFloat=&RemapRange(Input Value To Remap=0, Input Min=0, Input Max=1, Output Min=0, Output Max=1)) {isSetParameters=true}
   /SetVariables(Emitter.CameraFade:NiagaraFloat=/Niagara/DynamicInputs/Camera/CameraFade(
     Camera Query:NiagaraDataInterfaceCamera=DataInterface()
     Fade Offset:NiagaraFloat=100
     Fade Distance:NiagaraFloat=200
     Use Owner Position:NiagaraBool=true
    )) {isSetParameters=true}
  ]
  particleSpawnScript [
   &InitializeParticle(
    Color=<rgba 1 0.4304 0.2426 1>
    Lifetime=&MergeHelper_FloatSwitch(0=1, 1=0.18, 2=0.3, 3=?, Index=&SpawnGroup(), Num=::NewEnumerator1)
    Mesh Scale=&Lerp_Vector(A=<0.2 0.2 0.4>, B=<0.3 0.3 0.6>, Alpha=$User.Thruster)
    Mesh Scale Mode=::NewEnumerator4
    Position Offset=&MaskVector(
     Mask Source=0
     Mask Parameter=?
     Mask=2
     Masked Vector=&Lerp_Vector(A=<0 0 -12>, B=<0 0 -15>, Alpha=$User.Thruster)
     Passthrough Vector=<0 0 0>
    )
    Sprite Size Mode=::NewEnumerator3
    Uniform Sprite Size=50
    UsePositionOffset=true
    Write Material Random=true
   )
   &InitialMeshOrientation(
    Orientation Axis=<0 0 1>
    Orientation Vector=<0 0 -1>
    Rotation Coordinate Space=::NewEnumerator0
    Use Orientation Vector=true
    Use Rotation Vector=true
   ) {module=InitialMeshOrientation001}
   /SetVariables(
    Particles.Flame2Scale:Vector3f=&Multiply_Vector(A=$Particles.Scale, B=<1 1 1>)
    Particles.FlameCoreScale:Vector3f=&Multiply_Vector(A=$Particles.Scale, B=<0.65 0.6 0.6>)
   ) {isSetParameters=true}
   /SetVariables(Particles.VisibilityTag=&SpawnGroup()) {isSetParameters=true}
   /SetVariables(Particles.Flare Color:LinearColor=<rgba 3 1.0965 0.4688 1>, Particles.Flare Sprite Size:Vector2f=<120 120>) {isSetParameters=true}
   &CameraOffset(Camera Offset Amount=&MergeHelper_FloatSwitch(0=0, 1=0, 2=3, 3=?, Index=&SpawnGroup(), Num=::NewEnumerator1))
  ]
  particleUpdateScript [
   &ParticleState()
   &SolveForcesAndVelocity(Clamp Velocity=false, Limit Acceleration=false)
   &ScaleColor(
    Scale Alpha=&FloatFromCurve(FloatCurve=DataInterface() {propertyValues {Curve {keys [
         <at 0 1>
         <at 0.05 0.1>
         {arriveTangent=1.5287, leaveTangent=1.5287, time=0.1, value=1}
         <at 0.1808 0.3>
         {arriveTangent=-1.1287, leaveTangent=-1.1287, time=0.241, value=1}
         <at 0.3137 0.15>
         {arriveTangent=-0.4335, arriveTangentWeight=0.2916, leaveTangent=-0.4335, leaveTangentWeight=0.1147, time=0.3571, value=1}
         <at 0.4291 0.1>
         {arriveTangent=-0.7251, arriveTangentWeight=0.1918, leaveTangent=-0.7251, leaveTangentWeight=0.058, time=0.5086, value=1}
         <at 0.567 0>
         <at 0.6229 1>
         <at 0.6808 0>
         <at 0.7553 1>
         <at 0.8163 0>
         {arriveTangentWeight=0.0138, leaveTangentWeight=0.0174, time=0.8577, value=1}
         <at 0.9072 0>
         {arriveTangentWeight=0.0841, leaveTangentWeight=0.151, time=0.9394, value=1}
         {arriveTangentWeight=0.198, leaveTangentWeight=0.1375, time=0.9536, value=1}
         <at 1 0>
        ]}}})
    Scale RGB=&MakeVectorFromLinearColorRGB(Color=&LerpLinearColors(StartColor=<rgba 1.6812 0.6813 0.6813 1>, EndColor=<rgba 10 10 10 1>, LerpFactor=$User.Thruster))
    ScaleA=true
    ScaleRGB=true
   )
   /SetVariables(
    Particles.Flame2Color:LinearColor=&MakeLinearColorFromVectorAndFloat(
     "Vector (RGB)"=<1 1 1>
     "Float (Alpha)"=&FloatFromCurve(
      FloatCurve=DataInterface() {propertyValues {Curve {keys [
          <at 0 1>
          <at 0.05 0.1>
          {arriveTangent=1.5287, leaveTangent=1.5287, time=0.1, value=1}
          <at 0.1808 0.3>
          {arriveTangent=-1.1287, leaveTangent=-1.1287, time=0.241, value=1}
          <at 0.3137 0.15>
          {arriveTangent=-0.4335, arriveTangentWeight=0.2916, leaveTangent=-0.4335, leaveTangentWeight=0.1147, time=0.3571, value=1}
          <at 0.4291 0.1>
          {arriveTangent=-0.7251, arriveTangentWeight=0.1918, leaveTangent=-0.7251, leaveTangentWeight=0.058, time=0.5086, value=1}
          <at 0.567 0>
          <at 0.6229 1>
          <at 0.6808 0>
          <at 0.7553 1>
          <at 0.8163 0>
          {arriveTangentWeight=0.0138, leaveTangentWeight=0.0174, time=0.8577, value=1}
          <at 0.9072 0>
          {arriveTangentWeight=0.0841, leaveTangentWeight=0.151, time=0.9394, value=1}
          {arriveTangentWeight=0.198, leaveTangentWeight=0.1375, time=0.9536, value=1}
          <at 1 0>
         ]}}}
      Scale Curve=&Add_Float(A=1, B=$User.Thruster)
     )
    )
    Particles.FlameCoreColor:LinearColor=&MakeLinearColorFromVectorAndFloat(
     "Vector (RGB)"=&MakeVectorFromLinearColorRGB(Color=<rgba 1 0.0473 0 1>)
     "Float (Alpha)"=&FloatFromCurve(
      FloatCurve=DataInterface() {propertyValues {Curve {keys [
          <at 0 1>
          <at 0.05 0.1>
          {arriveTangent=1.5287, leaveTangent=1.5287, time=0.1, value=1}
          <at 0.1808 0.3>
          {arriveTangent=-1.1287, leaveTangent=-1.1287, time=0.241, value=1}
          <at 0.3137 0.15>
          {arriveTangent=-0.4335, arriveTangentWeight=0.2916, leaveTangent=-0.4335, leaveTangentWeight=0.1147, time=0.3571, value=1}
          <at 0.4291 0.1>
          {arriveTangent=-0.7251, arriveTangentWeight=0.1918, leaveTangent=-0.7251, leaveTangentWeight=0.058, time=0.5086, value=1}
          <at 0.567 0>
          <at 0.6229 1>
          <at 0.6808 0>
          <at 0.7553 1>
          <at 0.8163 0>
          {arriveTangentWeight=0.0138, leaveTangentWeight=0.0174, time=0.8577, value=1}
          <at 0.9072 0>
          {arriveTangentWeight=0.0841, leaveTangentWeight=0.151, time=0.9394, value=1}
          {arriveTangentWeight=0.198, leaveTangentWeight=0.1375, time=0.9536, value=1}
          <at 1 0>
         ]}}}
      Scale Curve=0.3
     )
    )
   ) {isSetParameters=true}
   &ScaleMeshSize(Scale Factor=&MakeVector(
     X=&FloatFromCurve(FloatCurve=DataInterface() {propertyValues {Curve {keys [{tangentMode=RCTM_User, time=0, value=1}, {tangentMode=RCTM_User, time=0.5, value=0.5}]}}})
     Y=&FloatFromCurve(FloatCurve=DataInterface() {propertyValues {Curve {keys [{tangentMode=RCTM_User, time=0, value=1}, {tangentMode=RCTM_User, time=0.5, value=0.5}]}}})
     Z=&FN_FrontEndFloatModifier(
      Gameplay=&FloatFromCurve(FloatCurve=DataInterface() {propertyValues {Curve {keys [<at 0 0.6>,<at 0.9 1>,<at 0.9993 0>]}}})
      Front End=0
     )
    ))
   /SetVariables(
    Particles.Flame2Scale:Vector3f=&Multiply_Vector(
     A=$Particles.Initial.Flame2Scale
     B=&Multiply_Vector(
      A=&MakeVector(
       X=1
       Y=1
       Z=&FloatFromCurve(FloatCurve=DataInterface() {propertyValues {Curve {keys [<at 0 0.6>,<at 0.97 1>,<at 1 0>]}}})
      )
      B=&Lerp_Vector(A=<1 1 0.7>, B=<1.3 1.3 1.3>, Alpha=$User.Thruster)
     )
    )
    Particles.FlameCoreScale:Vector3f=&Multiply_Vector(
     A=$Particles.Initial.FlameCoreScale
     B=&Multiply_Vector(
      A=&MakeVector(
       X=1
       Y=1
       Z=&FloatFromCurve(
        FloatCurve=DataInterface() {propertyValues {Curve {keys [<at 0 0.6>,<at 0.1 1>,<at 1 0>]}}}
        Scale Curve=&Lerp_Float(A=1, B=1.5, Alpha=$User.Thruster)
       )
      )
      B=&Lerp_Vector(A=<1 1 0.5>, B=<1.4 1.4 1.5>, Alpha=$User.Thruster)
     )
    )
   ) {isSetParameters=true}
   /SetVariables(
    Particles.Flame2Scale:Vector3f=&Multiply_Vector(
     A=$Particles.Flame2Scale
     B=&MakeVector(
      X=&FloatFromCurve(FloatCurve=DataInterface() {propertyValues {Curve {
          defaultValue=0.1
          keys [
           {arriveTangent=13.4258, interpMode=RCIM_Cubic, leaveTangent=13.4258, leaveTangentWeight=0.0667, tangentMode=RCTM_User, time=0, value=0.1}
           <at 0.2 1 RCIM_Constant>
           {arriveTangentWeight=0.0003, interpMode=RCIM_Cubic, tangentMode=RCTM_User, time=0.201, value=0.1}
           <at 0.4 1 RCIM_Constant>
           {arriveTangentWeight=0.0003, interpMode=RCIM_Cubic, tangentMode=RCTM_User, time=0.401, value=0.1}
           <at 0.6 1.0392 RCIM_Constant>
           {arriveTangentWeight=0.0003, interpMode=RCIM_Cubic, tangentMode=RCTM_User, time=0.601, value=0.1}
           <at 0.8 1.0392 RCIM_Constant>
           {arriveTangentWeight=0.0003, interpMode=RCIM_Cubic, tangentMode=RCTM_User, time=0.801, value=0.1}
           <at 1 1 RCIM_Constant>
          ]
         }}})
      Y=&FloatFromCurve(FloatCurve=DataInterface() {propertyValues {Curve {
          defaultValue=0.1
          keys [
           {arriveTangent=13.4258, interpMode=RCIM_Cubic, leaveTangent=13.4258, leaveTangentWeight=0.0667, tangentMode=RCTM_User, time=0, value=0.1}
           {interpMode=RCIM_Constant, tangentMode=RCTM_User, time=0.2, value=1}
           {interpMode=RCIM_Cubic, tangentMode=RCTM_User, time=0.201, value=0.1}
           <at 0.4 1 RCIM_Constant>
           {arriveTangent=10.1301, interpMode=RCIM_Cubic, leaveTangent=10.1301, tangentMode=RCTM_User, time=0.401, value=0.1}
           <at 0.6 1.0392 RCIM_Constant>
           {arriveTangent=10.1301, interpMode=RCIM_Cubic, leaveTangent=10.1301, tangentMode=RCTM_User, time=0.601, value=0.1}
           <at 0.8 1.0392 RCIM_Constant>
           {arriveTangent=10.1301, interpMode=RCIM_Cubic, leaveTangent=10.1301, tangentMode=RCTM_User, time=0.801, value=0.1}
           <at 1 1 RCIM_Constant>
          ]
         }}})
      Z=1
     )
    )
    Particles.FlameCoreScale:Vector3f=&Multiply_Vector(
     A=$Particles.FlameCoreScale
     B=&Multiply_Vector(
      A=&MakeVector(
       X=&FloatFromCurve(FloatCurve=DataInterface() {propertyValues {Curve {keys [{tangentMode=RCTM_User, time=0, value=1}, {tangentMode=RCTM_User, time=0.5, value=1}, <at 0.9093 1>]}}})
       Y=1
       Z=&FloatFromCurve(FloatCurve=DataInterface() {propertyValues {Curve {keys [<at 0 0.6>,<at 0.97 1>,<at 1 0>]}}})
      )
      B=&MakeVector(
       X=1
       Y=1
       Z=&FloatFromCurve(FloatCurve=DataInterface() {propertyValues {Curve {keys [<at 0 1>,<at 0.97 1>,<at 1 0>]}}})
      )
     )
    )
   ) {isSetParameters=true}
   &DynamicMaterialParameters(
    Index 0 Param 1=&Lerp_Float(A=0.3, B=0.6, Alpha=1)
    Module.Index0Float0Write=true
    Module.Index0Float1Write=true
    Module.Index0Float2Write=true
    Module.Index0Float3Write=true
   )
   /SetVariables(
    Particles.Wind Color:LinearColor=/Niagara/DynamicInputs/TypeConversions/MakeLinearColorFromFloat(Value:NiagaraFloat=&FloatFromCurve(FloatCurve=DataInterface() {propertyValues {Curve {
         defaultValue=1
         keys [
          {arriveTangent=9.0725, interpMode=RCIM_Cubic, leaveTangent=9.0725, tangentMode=RCTM_User, time=0, value=0}
          {arriveTangent=-2.5249, interpMode=RCIM_Cubic, leaveTangent=-2.5249, tangentMode=RCTM_User, time=0.1408, value=0.9847}
          {arriveTangent=-0.0405, interpMode=RCIM_Cubic, leaveTangent=-0.0405, tangentMode=RCTM_User, time=1, value=0}
         ]
        }}}))
    Particles.Wind Scale:Vector3f=<1 0.5 0.5>
   ) {isSetParameters=true}
   /SetVariables(
    Particles.Flare Color:LinearColor=&MakeLinearColorFromVectorAndFloat(
     "Vector (RGB)"=&MakeVectorFromLinearColorRGB(Color=&Multiply_LinearColorByFloat(Linear Color=$Particles.Initial.Flare Color, Float=1))
     "Float (Alpha)"=&FloatFromCurve(
      FloatCurve=DataInterface() {propertyValues {Curve {keys [
          {arriveTangent=-0.9, time=0, value=0}
          {arriveTangent=-0.0998, time=0.05, value=1}
          {arriveTangent=-1.0872, leaveTangent=-1.0872, time=0.0688, value=0}
          <at 0.1141 0.95>
          <at 0.1507 0>
          <at 0.1987 1>
          <at 0.2343 0>
          <at 0.277 1>
          <at 0.316 0>
          <at 0.3728 1>
          <at 0.4234 0>
          <at 0.5255 0.95>
          <at 0.5315 0>
          <at 0.6404 1>
          <at 0.6954 0>
          {arriveTangent=-1.5912, leaveTangent=-1.5912, time=0.7682, value=1}
          <at 0.8633 0>
          <at 0.9501 1>
          <at 1.0002 0>
         ]}}}
      Scale Curve=&Lerp_Float(A=0.7, B=3, Alpha=$User.Thruster)
     )
    )
    Particles.Flare Sprite Size:Vector2f=&Multiply_Vector2D(
     A=$Particles.Initial.Flare Sprite Size
     B=&Vector2DFromFloat(Value=&FloatFromCurve(
       FloatCurve=DataInterface() {propertyValues {Curve {keys [
           {interpMode=RCIM_Cubic, leaveTangentWeight=0.2638, tangentMode=RCTM_User, time=0, value=1}
           {arriveTangentWeight=0.3731, interpMode=RCIM_Cubic, leaveTangentWeight=0.0983, tangentMode=RCTM_User, time=0.8982, value=1}
           <at 1 0 RCIM_Cubic>
          ]}}}
       Scale Curve=&Lerp_Float(A=0.1, B=1, Alpha=$User.Thruster)
      ))
    )
   ) {isSetParameters=true}
   &FN_Light_Attributes(
    Color=&ColorFromCurve(ColorCurve=DataInterface() {propertyValues {
       AlphaCurve {keys [<at 0 1>,<at 0.8973 1>,<at 1 0>]}
       BlueCurve {keys [<at 0 2.501 RCIM_Cubic>]}
       GreenCurve {keys [<at 0 6.6555 RCIM_Cubic>]}
       RedCurve {keys [<at 0 12 RCIM_Cubic>]}
      }})
    Create Color Attribute=true
    Create Exponent Attribute=true
    Create Position Attribute=true
    Create Radius Attribute=true
    Enable Particle Light Enable=true
    Exponent=2
    Light Enabled=true
    Position=<60 30 10>
    Radius=100
    Scale Color=&FloatFromCurve(
     FloatCurve=DataInterface() {propertyValues {Curve {keys [
         {arriveTangent=4.4475, time=0, value=0}
         {arriveTangent=5.3518, leaveTangent=-3.1774, time=0.1, value=1.2}
         {arriveTangent=-1.4972, leaveTangent=-1.4972, time=0.1523, value=0.7}
         {arriveTangent=-0.1667, leaveTangent=-0.1667, time=0.2052, value=1.1711}
         <at 0.258 0.7713>
         <at 0.2926 1.1775>
         {arriveTangent=-0.25, leaveTangent=-0.25, time=0.3608, value=0.689}
         <at 0.3865 1.193>
         <at 0.4263 0.8643>
         <at 0.4931 1.1191>
         {arriveTangent=-0.25, leaveTangent=-0.25, time=0.5428, value=0.7548}
         <at 0.5846 1.1063>
         <at 0.6571 0.8076>
         <at 0.6976 1.1685>
         {arriveTangent=0.1667, leaveTangent=0.1667, time=0.7542, value=0.7713}
         <at 0.7906 1.1138>
         <at 0.8535 0.8807>
         {arriveTangent=-3.3333, leaveTangent=-3.3333, time=0.9, value=1.2}
         <at 1 0>
        ]}}}
     Scale Curve=&Multiply_Float(A=3, B=&FNTimeOfDayScalar(Night Scale=0.5, Morning Scale=0.8, Evening Scale=0.8))
    )
   )
  ]
  rendererClasses [&NiagaraMeshRendererProperties, &NiagaraSpriteRendererProperties]
  renderers [
   {rendererClass=&NiagaraMeshRendererProperties, rendererIndex=0}
   {rendererClass=&NiagaraMeshRendererProperties, rendererIndex=1}
   {rendererClass=&NiagaraMeshRendererProperties, rendererIndex=2}
   {rendererClass=&NiagaraMeshRendererProperties, rendererIndex=3}
   {rendererClass=&NiagaraSpriteRendererProperties, rendererIndex=4}
  ]
 }}
summary {
 systemName=NS_Emote_TideCompost_Thruster
 emitters [{emitterName=Missile_FX, rendererClasses [&NiagaraMeshRendererProperties, &NiagaraSpriteRendererProperties]}]
 userVariables [
  User.bIsFrontEnd {type=NiagaraBool}
  User.bIsFrontEndPreview {type=NiagaraBool}
  User.GlobalScale {type=NiagaraFloat, defaultValue=2, description=""}
  User.Thruster {type=NiagaraFloat, defaultValue=1, description=""}
 ]
}
